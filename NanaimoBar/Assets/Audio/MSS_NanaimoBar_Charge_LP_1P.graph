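# Stereo MetaSound source for the NanaimoBar first-person charge loop, routed to the ranged weapon fire 1P sound class.
# On play, a random pick from the Sound array starts a wave player, with pitch shifted by a random offset around a base plus slow LFO wobble.
# Both channels are volume-scaled; playback end fires the one-shot finish.
# The Loop default holds a stringified builder result, and concurrency is limited per owner.
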MetaSoundSource /NanaimoBar/Assets/Audio/MSS_NanaimoBar_Charge_LP_1P
inputs [
 Random.Pitch.Base {type=Float}
 Random.Pitch.Spread {type=Float}
 Random.Volume.Volume {type=Float}
 Sound {type=WaveAsset:Array}
 UE.Source.OnPlay {type=Trigger}
]
interfaces [UE.OutputFormat.Stereo:1.0, UE.Source.OneShot:1.0, UE.Source:1.1]
outputs [
 "UE.OutputFormat.Stereo.Audio:0" {type=Audio}
 "UE.OutputFormat.Stereo.Audio:1" {type=Audio}
 UE.Source.OneShot.OnFinished {type=Trigger}
]
pages [Default {
  input_defaults {
   Random.Pitch.Base=0
   Random.Pitch.Spread=1
   Random.Volume.Volume=0.75
   UE.Source.OnPlay=null
   Sound [&/NanaimoBar_Charge_LP_01]
  }
  nodes [
   n1 = UE.Add:Float(PrimaryOperand=^Random.Pitch.Base, AdditionalOperands=^Random.Pitch.Spread)
   n2 = UE.Reroute(Value=^UE.Source.OnPlay)
   n3 = UE.Subtract:Float(PrimaryOperand=^Random.Pitch.Base, AdditionalOperands=^Random.Pitch.Spread)
   ArrayRandomGet_0 = Array.Random Get:WaveAsset:Array(In Array=^Sound, Next=^UE.Source.OnPlay)
   WavePlayer_0 = UE.Wave Player:Stereo(
    Pitch Shift=UE.Add:Float(
     AdditionalOperands=UE.LFO:Audio() {also {defaults {Frequency=0.4}}}.Out
     PrimaryOperand=UE.RandomFloat(Max=@n1.Out, Next=@n2.Value, Min=@n3.Out).Value
    ).Out
    Play=@ArrayRandomGet_0.On Next
    Wave Asset=@ArrayRandomGet_0.Value
   ) {also {defaults {Loop="<MetaSoundBuilderResult.SUCCEEDED: 0>"}, version [1,1]}}
   Multiply_0 = UE.Multiply:Audio by Float(AdditionalOperands=^Random.Volume.Volume, PrimaryOperand=@WavePlayer_0.Out Left)
   Multiply_1 = UE.Multiply:Audio by Float(AdditionalOperands=^Random.Volume.Volume, PrimaryOperand=@WavePlayer_0.Out Right)
  ]
  links [
   {dst="@UE.OutputFormat.Stereo.Audio:0", src=Multiply_0.Out}
   {dst="@UE.OutputFormat.Stereo.Audio:1", src=Multiply_1.Out}
   {dst="@UE.Source.OneShot.OnFinished", src=WavePlayer_0.On Finished}
  ]
 }]
properties {
 assetImportData=&self:AssetImportData
 bOverrideConcurrency=true
 numChannels=2
 outputFormat=Stereo
 sampleRate=48000
 soundClassObject=&Fort_WeaponRanged_Fire_1P_PSM
 concurrencyOverrides {bLimitToOwner=true, maxCount=3}
}
