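# One-shot stereo MetaSound for the third-person NanaimoBar weapon charge start, with attenuation and a concurrency limit.
# An affiliation check on play either triggers immediately or also waits for ChargePercent to cross a low threshold.
# That trigger plays a random Charge Start wave alongside the Dry Fire input (null by default), whose gain is slightly randomized.
# An affiliation-based gain scales the mix, and the source finishes once both players finish.
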MetaSoundSource /NanaimoBar/Assets/Audio/MSS_NanaimoBar_Charge_Start_3P
inputs [
 Character.Affiliation {type=Enum:FortAudioAffiliation}
 Character.Health {type=Float}
 Character.UniqueID {type=Int32}
 Charge Start {type=WaveAsset:Array}
 ChargePercent {type=Float}
 Dry Fire {type=WaveAsset}
 UE.Source.OnPlay {type=Trigger}
]
interfaces [Fort.Character:1.0, UE.OutputFormat.Stereo:1.0, UE.Source.OneShot:1.0, UE.Source:1.1]
outputs [
 "UE.OutputFormat.Stereo.Audio:0" {type=Audio}
 "UE.OutputFormat.Stereo.Audio:1" {type=Audio}
 UE.Source.OneShot.OnFinished {type=Trigger}
]
pages [Default {
  input_defaults {
   Character.Affiliation=0
   Character.Health=0
   Character.UniqueID=0
   ChargePercent=0.11
   Dry Fire=null
   UE.Source.OnPlay=null
   Charge Start [
    &/NanaimoBar_Charge_Start_01
    &/NanaimoBar_Charge_Start_02
    &/NanaimoBar_Charge_Start_03
   ]
  }
  nodes [
   n1 = UE.Reroute(Value=UE.RandomFloat(Next=^UE.Source.OnPlay) {also {defaults {Max=1.05, Min=0.95}}}.Value)
   n2 = &TriggerCompareAffiliation(A=^Character.Affiliation, Compare=^UE.Source.OnPlay)
   TriggerAccumulatorTriggerAccumulate2_1 = "TriggerAccumulator.Trigger Accumulate (2)"(In 0=@n2.False, In 1=UE.TriggerOnThreshold:Float(In=^ChargePercent) {also {defaults {Threshold=0.1}}}.Out)
   TriggerAnyTriggerAny2_0 = "TriggerAny.Trigger Any (2)"(In 1=@TriggerAccumulatorTriggerAccumulate2_1.Out, In 0=@n2.True)
   n3 = VariableAccessor.Trigger(Variable=VariableMutator.Trigger(Variable=InitVariable.Trigger().Variable, Value=@TriggerAnyTriggerAny2_0.Out).Variable)
   WavePlayer_1 = UE.Wave Player:Stereo(Wave Asset=^Dry Fire, Play=VariableAccessor.Trigger(Variable=@n3.Variable).Value) {also {version [1,1]}}
   ArrayRandomGet_0 = Array.Random Get:WaveAsset:Array(In Array=^Charge Start, Next=@n3.Value)
   WavePlayer_0 = UE.Wave Player:Stereo(Play=@ArrayRandomGet_0.On Next, Wave Asset=@ArrayRandomGet_0.Value) {also {version [1,1]}}
   AudioMixerAudioMixerStereo2_0 = "AudioMixer.Audio Mixer (Stereo, 2)"(
    In 1 L=UE.Multiply:Audio by Float(AdditionalOperands=@n1.Value, PrimaryOperand=@WavePlayer_1.Out Left).Out
    In 1 R=UE.Multiply:Audio by Float(AdditionalOperands=@n1.Value, PrimaryOperand=@WavePlayer_1.Out Right).Out
    In 0 L=@WavePlayer_0.Out Left
    In 0 R=@WavePlayer_0.Out Right
   )
   n4 = &MSP_AffiliationMixFloat(Affiliation=^Character.Affiliation) {also {defaults {Non-Enemy=0.3}}}
   Multiply_1 = UE.Multiply:Audio by Float(PrimaryOperand=@AudioMixerAudioMixerStereo2_0.Out L, AdditionalOperands=@n4.Out)
   Multiply_2 = UE.Multiply:Audio by Float(PrimaryOperand=@AudioMixerAudioMixerStereo2_0.Out R, AdditionalOperands=@n4.Out)
   TriggerAccumulatorTriggerAccumulate2_0 = "TriggerAccumulator.Trigger Accumulate (2)"(In 0=@WavePlayer_0.On Finished, In 1=@WavePlayer_1.On Finished)
  ]
  links [
   {dst="@UE.OutputFormat.Stereo.Audio:0", src=Multiply_1.Out}
   {dst="@UE.OutputFormat.Stereo.Audio:1", src=Multiply_2.Out}
   {dst="@UE.Source.OneShot.OnFinished", src=TriggerAccumulatorTriggerAccumulate2_0.Out}
  ]
 }]
properties {
 assetImportData=&self:AssetImportData
 attenuationSettings=&/ATT_NanaimoBar_Charge_3P
 bOverrideConcurrency=true
 maxDistance=22750
 numChannels=2
 outputFormat=Stereo
 sampleRate=48000
 soundClassObject=&Fort_WeaponRanged_Fire_3P_PSM
 volume=0.35
 concurrencyOverrides {maxCount=3}
}
