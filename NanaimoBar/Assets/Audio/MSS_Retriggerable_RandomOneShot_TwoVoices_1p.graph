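# Retriggerable 5.1 one-shot MetaSound source that plays a random wave with random pitch around a base on each Retrigger.
# A trigger sequence alternates two summed wave players so consecutive sounds overlap, with gates passing only the latest voice's finish.
# That finish starts a delay that ends the source unless a retrigger closes its gate.
# The Mobile page uses one stereo player without pitch randomization.
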MetaSoundSource /NanaimoBar/Assets/Audio/MSS_Retriggerable_RandomOneShot_TwoVoices_1p
inputs [
 Random.Pitch.Base {type=Float}
 Random.Pitch.Spread {type=Float}
 Retrigger {type=Trigger}
 RetriggerDelayTime {type=Time}
 Sounds {type=WaveAsset:Array}
 UE.Source.OnPlay {type=Trigger}
]
interfaces [UE.OutputFormat.5dot1:1.0, UE.Source.OneShot:1.0, UE.Source:1.1]
outputs [
 bIsAlmostFinished {type=Bool}
 "UE.OutputFormat.5dot1.Audio:0" {type=Audio}
 "UE.OutputFormat.5dot1.Audio:1" {type=Audio}
 "UE.OutputFormat.5dot1.Audio:2" {type=Audio}
 "UE.OutputFormat.5dot1.Audio:3" {type=Audio}
 "UE.OutputFormat.5dot1.Audio:4" {type=Audio}
 "UE.OutputFormat.5dot1.Audio:5" {type=Audio}
 UE.Source.OneShot.OnFinished {type=Trigger}
]
pages [
 Default {
  is_build_graph=true
  input_defaults {Random.Pitch.Base=0, Random.Pitch.Spread=0, Retrigger=null, RetriggerDelayTime=2, UE.Source.OnPlay=null, Sounds []}
  nodes [
   ArrayRandomGet_0 = Array.Random Get:WaveAsset:Array(Next=^Retrigger, In Array=^Sounds) {also {defaults {No Repeats=-1, Weights []}}}
   n1 = AbsoluteValue.Abs:Float(Input=^Random.Pitch.Spread)
   n2 = UE.RandomFloat(
    Next=^Retrigger
    Max=UE.Add:Float(PrimaryOperand=^Random.Pitch.Base, AdditionalOperands=@n1.Value).Out
    Min=UE.Subtract:Float(PrimaryOperand=^Random.Pitch.Base, AdditionalOperands=@n1.Value).Out
   )
   TriggerSequence2_0 = "UE.Trigger Sequence (2)"(In=@ArrayRandomGet_0.On Next)
   WavePlayer_0 = UE.Wave Player:5dot1(Wave Asset=@ArrayRandomGet_0.Value, Pitch Shift=@n2.Value, Play=@TriggerSequence2_0.Out 0) {also {version [1,1]}}
   WavePlayer_1 = UE.Wave Player:5dot1(Wave Asset=@ArrayRandomGet_0.Value, Pitch Shift=@n2.Value, Play=@TriggerSequence2_0.Out 1) {also {version [1,1]}}
   Add_0 = UE.Add:Audio(PrimaryOperand=@WavePlayer_0.Out Front Center, AdditionalOperands=@WavePlayer_1.Out Front Center)
   Add_1 = UE.Add:Audio(PrimaryOperand=@WavePlayer_0.Out Side Right, AdditionalOperands=@WavePlayer_1.Out Side Right)
   Add_2 = UE.Add:Audio(PrimaryOperand=@WavePlayer_0.Out Side Left, AdditionalOperands=@WavePlayer_1.Out Side Left)
   Add_3 = UE.Add:Audio(PrimaryOperand=@WavePlayer_0.Out Low Frequency, AdditionalOperands=@WavePlayer_1.Out Low Frequency)
   Add_4 = UE.Add:Audio(PrimaryOperand=@WavePlayer_0.Out Front Right, AdditionalOperands=@WavePlayer_1.Out Front Right)
   Add_5 = UE.Add:Audio(PrimaryOperand=@WavePlayer_0.Out Front Left, AdditionalOperands=@WavePlayer_1.Out Front Left)
   TriggerControl_0 = UE.Trigger Control(Trigger In=@WavePlayer_0.On Finished, Open=@WavePlayer_0.On Play, Close=@WavePlayer_1.On Play)
   TriggerControl_1 = UE.Trigger Control(Close=@WavePlayer_0.On Play, Trigger In=@WavePlayer_1.On Finished, Open=@WavePlayer_1.On Play)
   TriggerAnyTriggerAny2_0 = "TriggerAny.Trigger Any (2)"(In 0=@TriggerControl_0.Trigger Out, In 1=@TriggerControl_1.Trigger Out)
   n3 = &MSP_TriggerResolve(B=^Retrigger, A=@TriggerAnyTriggerAny2_0.Out)
   TriggerDelay_0 = UE.Trigger Delay(Delay Time=^RetriggerDelayTime, In=@TriggerAnyTriggerAny2_0.Out)
   TriggerControl_2 = UE.Trigger Control(Open=@n3.A, Close=@n3.B, Trigger In=@TriggerDelay_0.Out)
   ValueValue_0 = Value.Value:Bool(Set=@TriggerControl_2.Trigger Out)
  ]
  links [
   {dst="@UE.OutputFormat.5dot1.Audio:2", src=Add_0.Out}
   {dst="@UE.OutputFormat.5dot1.Audio:5", src=Add_1.Out}
   {dst="@UE.OutputFormat.5dot1.Audio:4", src=Add_2.Out}
   {dst="@UE.OutputFormat.5dot1.Audio:3", src=Add_3.Out}
   {dst="@UE.OutputFormat.5dot1.Audio:1", src=Add_4.Out}
   {dst="@UE.OutputFormat.5dot1.Audio:0", src=Add_5.Out}
   {dst="@UE.Source.OneShot.OnFinished", src=ValueValue_0.On Set}
   {dst="@bIsAlmostFinished", src=ValueValue_0.Output Value}
  ]
 }
 Mobile {
  nodes [
   n1 = &MSP_TriggerResolve(B=^Retrigger)
   ArrayRandomGet_0 = Array.Random Get:WaveAsset:Array(Next=^Retrigger, In Array=^Sounds) {also {defaults {No Repeats=-1, Weights []}}}
   WavePlayer_0 = UE.Wave Player:Stereo(Play=@ArrayRandomGet_0.On Next, Wave Asset=@ArrayRandomGet_0.Value) {also {version [1,1]}}
   TriggerDelay_0 = UE.Trigger Delay(Delay Time=^RetriggerDelayTime, In=@WavePlayer_0.On Finished)
   TriggerControl_0 = UE.Trigger Control(Open=@n1.A, Close=@n1.B, Trigger In=@TriggerDelay_0.Out)
   ValueValue_0 = Value.Value:Bool(Set=@TriggerControl_0.Trigger Out)
  ]
  links [
   {dst="@UE.Source.OneShot.OnFinished", src=ValueValue_0.On Set}
   {dst="@bIsAlmostFinished", src=ValueValue_0.Output Value}
   {dst="@UE.OutputFormat.5dot1.Audio:0", src=WavePlayer_0.Out Left}
   {dst="@UE.OutputFormat.5dot1.Audio:1", src=WavePlayer_0.Out Right}
  ]
 }
]
properties {
 assetImportData=&self:AssetImportData
 numChannels=6
 outputFormat=FiveDotOne
 sampleRate=48000
}
