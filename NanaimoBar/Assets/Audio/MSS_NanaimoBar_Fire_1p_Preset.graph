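# MetaSound Source preset that plays the first-person firing sound for the NanaimoBar ranged weapon.
# It inherits its graph from MSS_Retriggerable_RandomOneShot_TwoVoices_1p, supplying three fire wave variations, random pitch settings and a retrigger delay.
# Output is a 5.1 one-shot using the first-person ranged-weapon fire sound class and local gunfire concurrency.
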MetaSoundSource /NanaimoBar/Assets/Audio/MSS_NanaimoBar_Fire_1p_Preset
preset_parent=&/MSS_Retriggerable_RandomOneShot_TwoVoices_1p
inputs [
 Random.Pitch.Base {type=Float}
 Random.Pitch.Spread {type=Float}
 Retrigger {type=Trigger}
 RetriggerDelayTime {type=Time}
 Sounds {type=WaveAsset:Array}
 UE.Source.OnPlay {type=Trigger}
]
interfaces [UE.OutputFormat.5dot1:1.0, UE.Source.OneShot:1.0, UE.Source:1.1]
outputs [
 bIsAlmostFinished {type=Bool}
 "UE.OutputFormat.5dot1.Audio:0" {type=Audio}
 "UE.OutputFormat.5dot1.Audio:1" {type=Audio}
 "UE.OutputFormat.5dot1.Audio:2" {type=Audio}
 "UE.OutputFormat.5dot1.Audio:3" {type=Audio}
 "UE.OutputFormat.5dot1.Audio:4" {type=Audio}
 "UE.OutputFormat.5dot1.Audio:5" {type=Audio}
 UE.Source.OneShot.OnFinished {type=Trigger}
]
pages [Default {input_defaults {
   Random.Pitch.Base=0
   Random.Pitch.Spread=3
   Retrigger=null
   RetriggerDelayTime=2
   UE.Source.OnPlay=null
   Sounds [
    &/NanaimoBar_Fire_1P_01
    &/NanaimoBar_Fire_1P_02
    &/NanaimoBar_Fire_1P_03
   ]
  }}]
properties {
 assetImportData=&self:AssetImportData
 numChannels=6
 outputFormat=FiveDotOne
 qualitySetting=Fort_Weapons
 sampleRate=48000
 soundClassObject=&Fort_WeaponRanged_Fire_1P_PSM
 volume=0.8
 concurrencySet [&Weapon_Gunfire_Local_Conc]
}
