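# Retriggerable stereo one-shot MetaSound using ranged-weapon distance curves: each Retrigger plays a random Close, Medium and Distant wave.
# Attenuation distance drives wavetable crossfade gains for the layers, then affiliation-based volume scales the mix.
# After all three finish, a delayed, retrigger-gated trigger signals completion.
# The Default page shares one random pitch across layers; Mobile uses fixed pitches and leaves the completion gate's reopen input unfed.
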MetaSoundSource /NanaimoBar/Assets/Audio/MSS_Retriggerable_RandomOneShot_TwoVoices_3p
inputs [
 Character.Affiliation {type=Enum:FortAudioAffiliation}
 Character.Health {type=Float}
 Character.UniqueID {type=Int32}
 MainFire.DistanceCurves {type=WaveTableBankAsset}
 MainFire.FalloffDistance {type=Float}
 Random.Pitch.Base {type=Float}
 Random.Pitch.Spread {type=Float}
 Retrigger {type=Trigger}
 RetriggerDelayTime {type=Time}
 Sounds.Close {type=WaveAsset:Array}
 Sounds.Distant {type=WaveAsset:Array}
 Sounds.Medium {type=WaveAsset:Array}
 UE.Attenuation.Distance {type=Float}
 UE.Source.OnPlay {type=Trigger}
 Volume.Close {type=Float}
 Volume.Distant {type=Float}
 Volume.Friendly {type=Float}
 Volume.Master {type=Float}
 Volume.Medium {type=Float}
]
interfaces [Fort.Character:1.0, UE.Attenuation:1.0, UE.OutputFormat.Stereo:1.0, UE.Source.OneShot:1.0, UE.Source:1.1]
outputs [
 bIsAlmostFinished {type=Bool}
 "UE.OutputFormat.Stereo.Audio:0" {type=Audio}
 "UE.OutputFormat.Stereo.Audio:1" {type=Audio}
 UE.Source.OneShot.OnFinished {type=Trigger}
]
pages [
 Default {
  is_build_graph=true
  input_defaults {
   Character.Affiliation=0
   Character.Health=0
   Character.UniqueID=0
   MainFire.DistanceCurves=&WeaponRanged_DistanceCurves
   MainFire.FalloffDistance=27500
   Random.Pitch.Base=0
   Random.Pitch.Spread=0
   Retrigger=null
   RetriggerDelayTime=2
   UE.Attenuation.Distance=27500
   UE.Source.OnPlay=null
   Volume.Close=1
   Volume.Distant=1
   Volume.Friendly=1
   Volume.Master=1
   Volume.Medium=1
   Sounds.Close []
   Sounds.Distant []
   Sounds.Medium []
  }
  nodes [
   InitVariable.Float()
   n1 = MapRange.MapRange:Float(In Range B=^MainFire.FalloffDistance, In=^UE.Attenuation.Distance) {also {defaults {Out Range A=0.001}}}
   ArrayRandomGet_2 = Array.Random Get:WaveAsset:Array(Next=^Retrigger, In Array=^Sounds.Distant) {also {defaults {No Repeats=-1, Weights []}}}
   n2 = AbsoluteValue.Abs:Float(Input=^Random.Pitch.Spread)
   n3 = UE.RandomFloat(
    Next=^Retrigger
    Max=UE.Add:Float(PrimaryOperand=^Random.Pitch.Base, AdditionalOperands=@n2.Value).Out
    Min=UE.Subtract:Float(PrimaryOperand=^Random.Pitch.Base, AdditionalOperands=@n2.Value).Out
   )
   WavePlayer_0 = UE.Wave Player:Stereo(Play=@ArrayRandomGet_2.On Next, Wave Asset=@ArrayRandomGet_2.Value, Pitch Shift=@n3.Value) {also {version [1,1]}}
   ArrayRandomGet_0 = Array.Random Get:WaveAsset:Array(Next=^Retrigger, In Array=^Sounds.Close) {also {defaults {No Repeats=-1, Weights []}}}
   WavePlayer_1 = UE.Wave Player:Stereo(Play=@ArrayRandomGet_0.On Next, Wave Asset=@ArrayRandomGet_0.Value, Pitch Shift=@n3.Value) {also {version [1,1]}}
   ArrayRandomGet_1 = Array.Random Get:WaveAsset:Array(Next=^Retrigger, In Array=^Sounds.Medium) {also {defaults {No Repeats=-1, Weights []}}}
   WavePlayer_2 = UE.Wave Player:Stereo(Play=@ArrayRandomGet_1.On Next, Wave Asset=@ArrayRandomGet_1.Value, Pitch Shift=@n3.Value) {also {version [1,1]}}
   AudioMixerAudioMixerStereo3_0 = "AudioMixer.Audio Mixer (Stereo, 3)"(
    Gain 2=UE.Multiply:Float(
     AdditionalOperands=^Volume.Distant
     PrimaryOperand=UE.WaveTableBankEvaluate(WaveTableBank=^MainFire.DistanceCurves, Input=@n1.Out Value) {also {defaults {Index=2}}}.Output
    ).Out
    Gain 1=UE.Multiply:Float(
     AdditionalOperands=^Volume.Medium
     PrimaryOperand=UE.WaveTableBankEvaluate(WaveTableBank=^MainFire.DistanceCurves, Input=@n1.Out Value) {also {defaults {Index=1}}}.Output
    ).Out
    Gain 0=UE.Multiply:Float(
     AdditionalOperands=^Volume.Close
     PrimaryOperand=UE.WaveTableBankEvaluate(WaveTableBank=^MainFire.DistanceCurves, Input=@n1.Out Value).Output
    ).Out
    In 2 L=@WavePlayer_0.Out Left
    In 2 R=@WavePlayer_0.Out Right
    In 0 L=@WavePlayer_1.Out Left
    In 0 R=@WavePlayer_1.Out Right
    In 1 L=@WavePlayer_2.Out Left
    In 1 R=@WavePlayer_2.Out Right
   )
   n4 = &MSP_AffiliationMixFloat(Affiliation=^Character.Affiliation, Non-Enemy=^Volume.Friendly, In Value=^Volume.Master)
   Multiply_0 = UE.Multiply:Audio by Float(PrimaryOperand=@AudioMixerAudioMixerStereo3_0.Out R, AdditionalOperands=@n4.Out)
   Multiply_1 = UE.Multiply:Audio by Float(PrimaryOperand=@AudioMixerAudioMixerStereo3_0.Out L, AdditionalOperands=@n4.Out)
   TriggerAccumulatorTriggerAccumulate3_0 = "TriggerAccumulator.Trigger Accumulate (3)"(In 2=@WavePlayer_0.On Finished, In 0=@WavePlayer_1.On Finished, In 1=@WavePlayer_2.On Finished) {also {defaults {Auto Reset="<MetaSoundBuilderResult.SUCCEEDED: 0>"}}}
   n5 = &MSP_TriggerResolve(B=^Retrigger, A=@TriggerAccumulatorTriggerAccumulate3_0.Out)
   TriggerDelay_0 = UE.Trigger Delay(Delay Time=^RetriggerDelayTime, In=@TriggerAccumulatorTriggerAccumulate3_0.Out)
   TriggerControl_0 = UE.Trigger Control(Open=@n5.A, Close=@n5.B, Trigger In=@TriggerDelay_0.Out)
   ValueValue_0 = Value.Value:Bool(Set=@TriggerControl_0.Trigger Out)
  ]
  links [
   {dst="@UE.OutputFormat.Stereo.Audio:1", src=Multiply_0.Out}
   {dst="@UE.OutputFormat.Stereo.Audio:0", src=Multiply_1.Out}
   {dst="@UE.Source.OneShot.OnFinished", src=ValueValue_0.On Set}
   {dst="@bIsAlmostFinished", src=ValueValue_0.Output Value}
  ]
 }
 Mobile {
  nodes [
   InitVariable.Float()
   n1 = MapRange.MapRange:Float(In Range B=^MainFire.FalloffDistance, In=^UE.Attenuation.Distance) {also {defaults {Out Range A=0.001}}}
   ArrayRandomGet_2 = Array.Random Get:WaveAsset:Array(Next=^Retrigger, In Array=^Sounds.Distant) {also {defaults {No Repeats=-1, Weights []}}}
   WavePlayer_0 = UE.Wave Player:Stereo(Play=@ArrayRandomGet_2.On Next, Wave Asset=@ArrayRandomGet_2.Value) {also {defaults {Pitch Shift=8}, version [1,1]}}
   ArrayRandomGet_0 = Array.Random Get:WaveAsset:Array(Next=^Retrigger, In Array=^Sounds.Close) {also {defaults {No Repeats=-1, Weights []}}}
   WavePlayer_1 = UE.Wave Player:Stereo(Play=@ArrayRandomGet_0.On Next, Wave Asset=@ArrayRandomGet_0.Value) {also {version [1,1]}}
   ArrayRandomGet_1 = Array.Random Get:WaveAsset:Array(Next=^Retrigger, In Array=^Sounds.Medium) {also {defaults {No Repeats=-1, Weights []}}}
   WavePlayer_2 = UE.Wave Player:Stereo(Play=@ArrayRandomGet_1.On Next, Wave Asset=@ArrayRandomGet_1.Value) {also {defaults {Pitch Shift=5}, version [1,1]}}
   AudioMixerAudioMixerStereo3_0 = "AudioMixer.Audio Mixer (Stereo, 3)"(
    Gain 1=UE.Multiply:Float(
     AdditionalOperands=^Volume.Medium
     PrimaryOperand=UE.WaveTableBankEvaluate(WaveTableBank=^MainFire.DistanceCurves, Input=@n1.Out Value) {also {defaults {Index=1}}}.Output
    ).Out
    Gain 2=UE.Multiply:Float(
     AdditionalOperands=^Volume.Distant
     PrimaryOperand=UE.WaveTableBankEvaluate(WaveTableBank=^MainFire.DistanceCurves, Input=@n1.Out Value) {also {defaults {Index=2}}}.Output
    ).Out
    Gain 0=UE.Multiply:Float(
     AdditionalOperands=^Volume.Close
     PrimaryOperand=UE.WaveTableBankEvaluate(WaveTableBank=^MainFire.DistanceCurves, Input=@n1.Out Value).Output
    ).Out
    In 2 L=@WavePlayer_0.Out Left
    In 2 R=@WavePlayer_0.Out Right
    In 0 L=@WavePlayer_1.Out Left
    In 0 R=@WavePlayer_1.Out Right
    In 1 L=@WavePlayer_2.Out Left
    In 1 R=@WavePlayer_2.Out Right
   )
   n2 = &MSP_AffiliationMixFloat(Affiliation=^Character.Affiliation, Non-Enemy=^Volume.Friendly, In Value=^Volume.Master)
   Multiply_0 = UE.Multiply:Audio by Float(PrimaryOperand=@AudioMixerAudioMixerStereo3_0.Out L, AdditionalOperands=@n2.Out)
   Multiply_1 = UE.Multiply:Audio by Float(PrimaryOperand=@AudioMixerAudioMixerStereo3_0.Out R, AdditionalOperands=@n2.Out)
   n3 = &MSP_TriggerResolve(B=^Retrigger)
   TriggerAccumulatorTriggerAccumulate3_0 = "TriggerAccumulator.Trigger Accumulate (3)"(In 2=@WavePlayer_0.On Finished, In 0=@WavePlayer_1.On Finished, In 1=@WavePlayer_2.On Finished) {also {defaults {Auto Reset="<MetaSoundBuilderResult.SUCCEEDED: 0>"}}}
   TriggerDelay_0 = UE.Trigger Delay(Delay Time=^RetriggerDelayTime, In=@TriggerAccumulatorTriggerAccumulate3_0.Out)
   TriggerControl_0 = UE.Trigger Control(Open=@n3.A, Close=@n3.B, Trigger In=@TriggerDelay_0.Out)
   ValueValue_0 = Value.Value:Bool(Set=@TriggerControl_0.Trigger Out)
  ]
  links [
   {dst="@UE.OutputFormat.Stereo.Audio:0", src=Multiply_0.Out}
   {dst="@UE.OutputFormat.Stereo.Audio:1", src=Multiply_1.Out}
   {dst="@UE.Source.OneShot.OnFinished", src=ValueValue_0.On Set}
   {dst="@bIsAlmostFinished", src=ValueValue_0.Output Value}
  ]
 }
]
properties {
 assetImportData=&self:AssetImportData
 numChannels=2
 outputFormat=Stereo
 sampleRate=48000
}
