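# MetaSound preset that plays the NanaimoBar weapon's third-person fire sound in stereo.
# It inherits its graph from MSS_Retriggerable_RandomOneShot_TwoVoices_3p and supplies close, medium and distant sample arrays, per-layer volumes and pitch randomization.
# Inputs expose attenuation distance with shared ranged-weapon distance curves, a retrigger with delay, and character affiliation alongside a friendly volume.
# Non-attenuated projectile attenuation settings, a 3P fire sound class and local and global gunfire concurrency govern playback.
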MetaSoundSource /NanaimoBar/Assets/Audio/MSS_NanaimoBar_Fire_3p_Preset
preset_parent=&/MSS_Retriggerable_RandomOneShot_TwoVoices_3p
inputs [
 Character.Affiliation {type=Enum:FortAudioAffiliation}
 Character.Health {type=Float}
 Character.UniqueID {type=Int32}
 MainFire.DistanceCurves {type=WaveTableBankAsset}
 MainFire.FalloffDistance {type=Float}
 Random.Pitch.Base {type=Float}
 Random.Pitch.Spread {type=Float}
 Retrigger {type=Trigger}
 RetriggerDelayTime {type=Time}
 Sounds.Close {type=WaveAsset:Array}
 Sounds.Distant {type=WaveAsset:Array}
 Sounds.Medium {type=WaveAsset:Array}
 UE.Attenuation.Distance {type=Float}
 UE.Source.OnPlay {type=Trigger}
 Volume.Close {type=Float}
 Volume.Distant {type=Float}
 Volume.Friendly {type=Float}
 Volume.Master {type=Float}
 Volume.Medium {type=Float}
]
interfaces [Fort.Character:1.0, UE.Attenuation:1.0, UE.OutputFormat.Stereo:1.0, UE.Source.OneShot:1.0, UE.Source:1.1]
outputs [
 bIsAlmostFinished {type=Bool}
 "UE.OutputFormat.Stereo.Audio:0" {type=Audio}
 "UE.OutputFormat.Stereo.Audio:1" {type=Audio}
 UE.Source.OneShot.OnFinished {type=Trigger}
]
pages [Default {input_defaults {
   Character.Affiliation=0
   Character.Health=0
   Character.UniqueID=0
   MainFire.DistanceCurves=&WeaponRanged_DistanceCurves
   MainFire.FalloffDistance=27500
   Random.Pitch.Base=0
   Random.Pitch.Spread=3
   Retrigger=null
   RetriggerDelayTime=2
   UE.Attenuation.Distance=27500
   UE.Source.OnPlay=null
   Volume.Close=1
   Volume.Distant=1
   Volume.Friendly=0.5
   Volume.Master=0.5
   Volume.Medium=0.7
   Sounds.Close [&/NanaimoBar_Fire_3P_Close_01]
   Sounds.Distant [
    &/NanaimoBar_Fire_3P_Distant_01
    &/NanaimoBar_Fire_3P_Distant_02
    &/NanaimoBar_Fire_3P_Distant_03
   ]
   Sounds.Medium [
    &/NanaimoBar_Fire_3P_Medium_01
    &/NanaimoBar_Fire_3P_Medium_02
    &/NanaimoBar_Fire_3P_Medium_03
   ]
  }}]
properties {
 assetImportData=&self:AssetImportData
 attenuationSettings=&Fort_Weapon_Projectile_NonAtt_Athena_Att
 maxDistance=27500
 numChannels=2
 outputFormat=Stereo
 qualitySetting=Fort_Weapons
 sampleRate=48000
 soundClassObject=&Fort_WeaponRanged_Fire_3P_PSM
 concurrencySet [&Weapon_Gunfire_Local_Conc, &Weapon_Gunfire_Global_Conc]
}
