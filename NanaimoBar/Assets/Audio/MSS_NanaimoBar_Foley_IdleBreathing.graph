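# One-shot mono MetaSound Source for NanaimoBar idle breathing foley, routed to a weapon foley sound class with its own attenuation asset.
# On play, it picks a non-repeating random breath from six waves and plays it with pitch shift randomized around base plus or minus spread, which is zero by default.
# A fixed volume input scales the output, and wave completion fires OnFinished.
# Concurrency is overridden to one instance per owner.
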MetaSoundSource /NanaimoBar/Assets/Audio/MSS_NanaimoBar_Foley_IdleBreathing
inputs [
 Random.Pitch.Base {type=Float}
 Random.Pitch.Spread {type=Float}
 Random.Volume.Volume Base {type=Float}
 Sounds {type=WaveAsset:Array}
 UE.Source.OnPlay {type=Trigger}
]
interfaces [UE.OutputFormat.Mono:1.0, UE.Source.OneShot:1.0, UE.Source:1.1]
outputs [
 "UE.OutputFormat.Mono.Audio:0" {type=Audio}
 UE.Source.OneShot.OnFinished {type=Trigger}
]
pages [Default {
  input_defaults {
   Random.Pitch.Base=0
   Random.Pitch.Spread=0
   Random.Volume.Volume Base=0.5
   UE.Source.OnPlay=null
   Sounds [
    &/NanaimoBar_Foley_IdleBreathing_01
    &/NanaimoBar_Foley_IdleBreathing_02
    &/NanaimoBar_Foley_IdleBreathing_03
    &/NanaimoBar_Foley_IdleBreathing_04
    &/NanaimoBar_Foley_IdleBreathing_05
    &/NanaimoBar_Foley_IdleBreathing_06
   ]
  }
  nodes [
   ArrayRandomGet_0 = Array.Random Get:WaveAsset:Array(In Array=^Sounds, Next=^UE.Source.OnPlay) {also {defaults {No Repeats=4, Weights []}}}
   WavePlayer_0 = UE.Wave Player:Mono(
    Play=@ArrayRandomGet_0.On Next
    Wave Asset=@ArrayRandomGet_0.Value
    Pitch Shift=UE.RandomFloat(
     Next=^UE.Source.OnPlay
     Max=UE.Add:Float(PrimaryOperand=^Random.Pitch.Base, AdditionalOperands=^Random.Pitch.Spread).Out
     Min=UE.Subtract:Float(PrimaryOperand=^Random.Pitch.Base, AdditionalOperands=^Random.Pitch.Spread).Out
    ).Value
   ) {also {version [1,1]}}
   Multiply_0 = UE.Multiply:Audio by Float(AdditionalOperands=^Random.Volume.Volume Base, PrimaryOperand=@WavePlayer_0.Out Mono)
  ]
  links [{dst="@UE.OutputFormat.Mono.Audio:0", src=Multiply_0.Out}, {dst="@UE.Source.OneShot.OnFinished", src=WavePlayer_0.On Finished}]
 }]
properties {
 assetImportData=&self:AssetImportData
 attenuationSettings=&/ATT_NanaimoBar_Foley_IdleBreathing
 bOverrideConcurrency=true
 maxDistance=400
 sampleRate=48000
 soundClassObject=&Fort_WeaponFoley
 concurrencyOverrides {bLimitToOwner=true, maxCount=1}
}
