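# Translucent, two-sided, lit material for a glowing animated film effect on Niagara mesh particles.
# Panning noise, distorted caustics and inky smoke build tinted color layers, blended toward a flat color at the U edges.
# The film texture masks that glow, feeding an emissive-adjustment function, and also tints a base color.
# It also cuts into particle alpha with view-angle and depth fade; further branches compute desaturation, top fade and a Fresnel smoke mask.
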Material /NanaimoBar/Assets/Fx/Materials/M_NanaimoBar_Film
graph {
 properties {
  bUsedWithNiagaraMeshParticles=true
  blendMode=BLEND_Translucent
  pixelDepthOffsetMode=PDOM_Legacy
  twoSided=true
  usedShadingModels=Default Lit
 }
 nodes [
  n1 = ParticleColor()
  n2 = MaterialFunctionCall("FadeSensitivity (optional)"=ScalarParameter() {parameterName="FadeSensitivity (optional) (S)"}) {materialFunction=&FadeBasedOnViewAngle}
  n3 = Multiply(A=@n1.A, B=@n2.Result)
  n4 = TextureSample() {texture=/NanaimoBar/Assets/Fx/Misc/T_NanaimoBar_Film}
  n5 = Multiply(A=OneMinus(None=@n4.R), B=ScalarParameter() {defaultValue=0.5, parameterName=Mask})
  DepthFade(Opacity=Saturate(None=Subtract(A=@n3, B=@n5)), FadeDistance=ScalarParameter() {parameterName=FadeDistance})
  n6 = MaterialFunctionCall(XY=TextureCoordinate() {uTiling=2, vTiling=1.5}) {materialFunction=&Swizzle}
  n7 = ComponentMask(None=VectorParameter() {parameterName=UV Scale, defaultValue=<rgba 1 1 1 0>}.RGB) {g=true, r=true}
  n8 = Panner(Coordinate=Add(A=Multiply(A=@n6.YX, B=@n7)) {constB=0}) {speedY=0.1}
  n9 = Multiply(
   A=VectorParameter() {parameterName=Color1, defaultValue=<rgba 3 0 0.05 1>}.RGB
   B=Power(Base=TextureSample(UVs=@n8) {texture=/Game/Effects/Fort_Effects/Textures/Tile/T_Noise_02}.R) {constExponent=0.8}
  )
  n10 = VectorParameter() {parameterName=Color3, defaultValue=<rgba 1 0 0.05 1>}
  n11 = MaterialFunctionCall(XY=TextureCoordinate() {vTiling=2}) {materialFunction=&Swizzle}
  n12 = MaterialFunctionCall(XY=TextureCoordinate() {uTiling=2, vTiling=0.5}) {materialFunction=&Swizzle}
  n13 = Multiply(A=TextureSample(UVs=Panner(Coordinate=@n12.YX) {speedY=-0.1}) {samplerType=SAMPLERTYPE_LinearGrayscale, texture=/Game/Effects/Fort_Effects/Emotes/UltraLight/Texture/T_DirectionalNoise_Greyscale}.R) {constB=0.1}
  n14 = TextureSample(UVs=Panner(Coordinate=Add(A=@n11.YX, B=@n13) {constB=0}) {speedY=0.05}) {samplerType=SAMPLERTYPE_LinearGrayscale, texture=/Game/Global/Textures/Default/Caustics/T_Nosie_Caustics_Grayscale}
  n15 = LinearInterpolate(
   A=VectorParameter() {parameterName=Color2, defaultValue=<rgba 0.0014 0 0.0885 1>}.RGB
   B=@n10.RGB
   Alpha=Saturate(None=LinearInterpolate(Alpha=@n14.R) {constA=-0.1})
  )
  n16 = Multiply(B=Multiply(A=TextureSample() {texture=&T_Inky_Smoke_Tile}.R, B=Power(Base=@n15) {constExponent=1})) {constA=1}
  n17 = VectorParameter() {parameterName=Color4, defaultValue=<rgba 0.0094 0 0.0625 1>}
  n18 = Multiply(A=Add(A=Add(A=@n9, B=@n16), B=@n17.RGB) {constB=0}, B=ScalarParameter() {defaultValue=1, parameterName=Top Brightness}) {constB=7}
  Desaturation(None=@n18, Fraction=ScalarParameter() {defaultValue=1, parameterName=Des})
  n19 = Multiply(A=@n10.RGB, B=Multiply(A=@n4.R, B=@n4.R))
  n20 = VectorParameter() {defaultValue=<rgba 1 0.0372 0.0936 1>}
  n21 = ComponentMask(None=TextureCoordinate()) {r=true}
  n22 = Multiply(A=@n21) {constB=2}
  n23 = Multiply(A=OneMinus(None=@n21)) {constB=2}
  n24 = OneMinus(None=Multiply(A=Saturate(None=@n22), B=Saturate(None=@n23)))
  n25 = Multiply(A=Saturate(None=Add(A=@n4.R, B=@n20.RGB)), B=Multiply(A=LinearInterpolate(A=@n18, B=@n17.RGB, Alpha=@n24), B=@n1.RGB))
  n26 = ScalarParameter() {defaultValue=0.2, parameterName=spow}
  MaterialFunctionCall(Emissive=Multiply(A=Add(A=@n19, B=@n25), B=Power(Base=@n4.R, Exp=@n26) {constExponent=0.5})) {materialFunction=&MF_FX_EmissiveAdjustments}
  n27 = TextureCoordinate()
  n28 = Multiply(A=Power(Base=ComponentMask(None=@n27) {g=true})) {constB=4}
  Multiply(A=@n3, B=LinearInterpolate(B=Saturate(None=@n28), Alpha=ScalarParameter() {parameterName=Use Top Fade}) {constA=1})
  n29 = Panner(Coordinate=MaterialFunctionCall(XY=TextureCoordinate() {uTiling=2, vTiling=0.5}) {materialFunction=&Swizzle}.YX) {speedY=0.07}
  n30 = Add(
   A=Power(Base=TextureSample(UVs=@n29) {texture=&T_Inky_Smoke_Tile}.R) {constExponent=1}
   B=TextureSample(UVs=@n29) {samplerType=SAMPLERTYPE_Grayscale, texture=&LowResBlurredNoise}.R
  )
  n31 = OneMinus(None=Fresnel() {baseReflectFraction=0})
  n32 = OneMinus(None=ComponentMask(None=TextureCoordinate()) {r=true})
  n33 = ScalarParameter() {defaultValue=4.5, parameterName=Bottom mask power}
  n34 = Saturate(None=Multiply(A=Power(Base=@n32, Exp=@n33) {constExponent=4.5}, B=ScalarParameter() {defaultValue=100, parameterName=Bottom mask contrast}) {constB=100})
  Multiply(A=Multiply(A=Multiply(A=@n1.A, B=@n30), B=Power(Base=@n31) {constExponent=3}), B=Multiply(A=Multiply(A=@n34), B=@n2.Result))
  Multiply(A=VectorParameter() {parameterName=BaseColor, defaultValue=<rgba 0.0155 0.0133 0.0365 1>}.RGB, B=@n4.R)
 ]
}
