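# Masked effects material for Niagara and mesh particles that erodes a surface with glowing hot edges and vertex displacement.
# Triplanar noise in pre-skinned local space, with an optional subpattern, combines with a radial sphere mask to form the erosion mask.
# That mask feeds opacity, a hardened emissive edge band and normal-direction WPO over a simple PBR base.
# Dynamic parameters expose texture scale, erosion, WPO amount and emissive color and intensity.
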Material /NanaimoBar/Assets/Fx/Materials/M_Nanaimo_Erosion_HotEdges
graph {
 properties {
  bAutomaticallySetUsageInEditor=false
  bUseMaterialAttributes=true
  bUsedWithMeshParticles=true
  bUsedWithNanite=true
  bUsedWithNiagaraMeshParticles=true
  bUsedWithNiagaraRibbons=true
  bUsedWithNiagaraSprites=true
  bUsedWithParticleSprites=true
  blendMode=BLEND_Masked
  opacityMaskClipValue=0.16
  pixelDepthOffsetMode=PDOM_Legacy
  previewMesh=/ProtoSwarm/NPCs/Swarm/TidalCrane_Swarm_WPOTest/Meshes/SM_BugProto_TA2
  usedShadingModels=Default Lit
 }
 nodes [
  n1 = Time()
  ConstantBiasScale(None=Sine(None=Divide(A=@n1) {constB=5}))
  n2 = LinearInterpolate(
   A=VectorParameter() {group=PBR, parameterName=BaseColor1, sortPriority=0, defaultValue=<rgba 0.0138 0.0135 0.0226 1>}.RGB
   B=VectorParameter() {group=PBR, parameterName=BaseColor2, sortPriority=0, defaultValue=<rgba 0.0052 0.0052 0.0052 1>}.RGB
   Alpha=NamedRerouteUsage()
  )
  n3 = ScalarParameter() {defaultValue=0.15, group=PBR, parameterName=Metallic, sortPriority=0}
  n4 = ScalarParameter() {defaultValue=0.8, group=PBR, parameterName=Specular, sortPriority=0}
  n5 = ScalarParameter() {defaultValue=0.45, group=PBR, parameterName=Roughness, sortPriority=0}
  n6 = GetMaterialAttributes(None=MaterialFunctionCall(MA In=MakeMaterialAttributes!(
     BaseColor=@n2
     Metallic=@n3
     Specular=@n4
     Roughness=@n5
     Anisotropy
     EmissiveColor
     Opacity
     OpacityMask
     Normal
     Tangent
     WorldPositionOffset
     SubsurfaceColor
     ClearCoat
     ClearCoatRoughness
     AmbientOcclusion
     Refraction
     CustomizedUVs_0
     CustomizedUVs_1
     CustomizedUVs_2
     CustomizedUVs_3
     CustomizedUVs_4
     CustomizedUVs_5
     CustomizedUVs_6
     CustomizedUVs_7
     PixelDepthOffset
     ShadingModel
     Displacement
    )) {materialFunction=&MF_CreativeMoveTool}.MA Out) {attributeGetTypes [
    #1
    #2
    #3
    #4
    #5
   ]}
  n7 = NamedRerouteUsage()
  n8 = SetMaterialAttributes!(
   MaterialAttributes=@n6.MaterialAttributes
   Base Color
   Specular=LinearInterpolate(B=@n6.Specular, Alpha=NamedRerouteUsage())
   Roughness=LinearInterpolate(B=@n6.Roughness, Alpha=NamedRerouteUsage()) {constA=0.8}
   Opacity Mask=@n7
   World Position Offset=NamedRerouteUsage()
   Opacity=@n7
   Emissive Color=NamedRerouteUsage()
  ) {attributeSetTypes [
    #1
    #2
    #3
    #4
    #5
    #6
    #7
   ]}
  MaterialFunctionCall!(In=@n8) {materialFunction=/Game/Effects/Niagara/MasterMaterials/Functions/MF_FX_MAEmissiveAdjustments}
  n9 = SphereMask(
   A=NamedRerouteUsage()
   B=VectorParameter() {group=Erosion, parameterName=RadialErosionOrigin_LS, defaultValue=<rgba 0 0 0 1>}.RGB
   Radius=ObjectRadius()
  ) {hardnessPercent=0}
  n10 = NamedRerouteUsage()
  n11 = OneMinus(None=@n10)
  n12 = Add(A=@n10, B=Multiply(A=Subtract(A=@n9, B=@n11), B=ScalarParameter() {defaultValue=0.4, group=Erosion, parameterName=Erode_SphereMask}))
  n13 = Subtract(A=Add(A=NamedRerouteUsage()), B=Multiply(A=Saturate(None=@n12)) {constB=2})
  n14 = SmoothStep(
   Min=Constant() {r=0.5}
   Max=Constant()
   Value=NamedRerouteDeclaration(None=Saturate(None=@n13)) {name=Erosion Mask_Saturated, nodeColor=<rgba 0.1484 0 1 1>}
  )
  n15 = VertexNormalWS()
  NamedRerouteDeclaration(None=Multiply(A=Multiply(A=@n14, B=@n15), B=NamedRerouteUsage())) {name=WPO, nodeColor=<rgba 0 0.7656 1 1>}
  n16 = DynamicParameter() {defaultValue=<rgba 1 1 0 5>, paramNames [Texture Scale, Emissive Intensity, Erode, ErodeWPO]}
  NamedRerouteDeclaration(None=@n16.Emissive Intensity) {name=DP_EmissiveIntensity, nodeColor=<rgba 1 0.9375 0 1>}
  NamedRerouteDeclaration(None=@n13) {name=Erosion Mask, nodeColor=<rgba 0 1 0.5547 1>}
  NamedRerouteDeclaration(None=DynamicParameter() {parameterIndex=1, defaultValue=<rgba 1.3999 0 5 1>, paramNames [R,G,B,A]}.RGB) {name=DP_EmissiveColor, nodeColor=<rgba 0.7344 0 1 1>}
  NamedRerouteDeclaration(None=@n16.Erode) {name=DP_Erode, nodeColor=<rgba 1 0.9375 0 1>}
  NamedRerouteDeclaration(None=@n16.ErodeWPO) {name=DP_ErodeWPO, nodeColor=<rgba 0.0312 0 1 1>}
  NamedRerouteDeclaration(None=@n9) {name=Debug, nodeColor=<rgba 0 1 0.5078 1>}
  n17 = MaterialFunctionCall!(
   UV Scale=Multiply(A=ScalarParameter() {defaultValue=58, group=Erosion, parameterName=ErosionTexture_Subpattern_Scale}, B=NamedRerouteUsage())
   Seam Sharpness=Constant() {r=3}
  ) {materialFunction=/Game/Athena/Environments/MaterialFunctions/MF_TriplanarFlat}
  n18 = TextureSample(UVs=@n17.UVs, Tex=NamedRerouteUsage()) {samplerSource=SSM_Wrap_WorldGroupSettings}
  n19 = LocalPosition() {localOrigin=InstancePreSkinning}
  n20 = VertexInterpolator(VS=@n19.XYZ)
  n21 = Divide(
   A=NamedRerouteDeclaration(None=ShaderStageSwitch(PixelShader=@n20.PS, VertexShader=@n19.XYZ)) {name=PreskinedLLocalPos, nodeColor=<rgba 1 0.3516 0 1>}
   B=Multiply(A=ScalarParameter() {defaultValue=500, group=Erosion, parameterName=ErosionTexture_1_Scale}, B=NamedRerouteUsage())
  )
  n22 = ComponentMask(None=@n21) {b=true, r=true}
  n23 = NamedRerouteUsage()
  n24 = ComponentMask(None=@n21) {b=true, g=true}
  n25 = NamedRerouteUsage()
  n26 = Normalize(VectorInput=Transform(None=VertexNormalWS()) {transformSourceType=TRANSFORMSOURCE_World, transformType=TRANSFORM_Local})
  n27 = ScalarParameter() {defaultValue=0.5, group=Erosion, parameterName=Triplanar_TransitionContrast}
  n28 = MaterialFunctionCall(In=Abs(None=ComponentMask(None=@n26) {r=true}), Contrast=@n27) {materialFunction=&CheapContrast}
  n29 = Abs(None=ComponentMask(None=@n26) {b=true})
  n30 = LinearInterpolate(
   A=LinearInterpolate(
    A=TextureSample(UVs=@n22, Tex=@n23) {samplerSource=SSM_Wrap_WorldGroupSettings}.R
    B=TextureSample(UVs=@n24, Tex=@n25) {samplerSource=SSM_Wrap_WorldGroupSettings}.R
    Alpha=ComponentMask(None=@n28.Result) {r=true}
   ) {desc=Float3}
   B=TextureSample(UVs=ComponentMask(None=@n21) {g=true, r=true}, Tex=NamedRerouteUsage()) {samplerSource=SSM_Wrap_WorldGroupSettings}.R
   Alpha=ComponentMask(None=MaterialFunctionCall(In=@n29, Contrast=@n27) {materialFunction=&CheapContrast}.Result) {r=true}
  )
  n31 = ScalarParameter() {defaultValue=1, group=Erosion, parameterName=ErosionTexture_Subpattern}
  n32 = Add(A=@n30, B=Multiply(A=LinearInterpolate(A=@n18.R, B=@n30, Alpha=@n17.Seam Mask), B=@n31))
  n33 = Add(A=@n31)
  NamedRerouteDeclaration(None=StaticSwitch(True=Divide(A=@n32, B=@n33), False=@n30, Value=StaticBoolParameter() {group=Erosion, parameterName=ErosionTexture_Subpattern})) {name=T_ErosionNoise, nodeColor=<rgba 0 0 1 1>}
  n34 = ScalarParameter() {defaultValue=1, group=Erosion, parameterName=Erosion_EmisiveEdge_Hardness}
  n35 = NamedRerouteUsage()
  n36 = LinearInterpolate(
   A=Subtract(B=@n34) {constA=0, desc=Low negative value}
   B=Add(A=@n34) {desc=High Positive}
   Alpha=Subtract(A=Add(A=@n35) {constB=0.35}, B=ScalarParameter() {defaultValue=0.2, group=Erosion, parameterName=Erosion_EmissveEdge_Size}) {constB=-1}
  ) {desc=Lerps between greater range of value}
  n37 = Power(Base=Saturate(None=OneMinus(None=@n36)), Exp=ScalarParameter() {defaultValue=6, group=Erosion, parameterName=Erosion_EmisiveEdge_Pow})
  n38 = NamedRerouteUsage()
  NamedRerouteDeclaration(None=Multiply(A=Multiply(A=@n37, B=@n38), B=NamedRerouteUsage())) {name=Emissive_Hot Edges, nodeColor=<rgba 0.1719 1 0 1>}
  NamedRerouteDeclaration(None=@n16.Texture Scale) {name=DP_TextureScale, nodeColor=<rgba 0.3828 1 0 1>}
  NamedRerouteDeclaration(None=TextureObjectParameter() {group=Erosion, parameterName=Erode Texture, texture=/ProtoSwarm/Assets/FX/T_SwarmerErosionMask}) {name=T_In_ErosionNoise, nodeColor=<rgba 0 1 0.25 1>}
  ParticleSize()
  Reroute()
  Reroute()
  ScalarParameter() {defaultValue=5, parameterName=Param_1}
  ScalarParameter() {defaultValue=0.58, group=Erosion, parameterName=Erode}
  VectorParameter() {group=Erosion, parameterName=Erosion_HotEdgeColor, defaultValue=<rgba 1.3999 0 5 1>}
 ]
}
