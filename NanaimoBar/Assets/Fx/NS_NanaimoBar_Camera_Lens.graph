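# Niagara system for the NanaimoBar camera lens effect, drawing a single mesh particle that persists past its short lifetime.
# One burst spawns the oriented, scaled mesh, and curves drive its size and a dynamic material parameter.
# Alpha multiplies a system-level Flicker value, a time-of-day scalar that is dimmer at night, and a fade-in curve indexed by particle age.
# An RGB color curve is present but switched off.
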NiagaraSystem /NanaimoBar/Assets/Fx/NS_NanaimoBar_Camera_Lens
dependencies {
 usedDataInterfaces [&NiagaraDataInterfaceCurve, &NiagaraDataInterfaceColorCurve]
 usedDynamicInputs [&Waveform]
 usedModules [
  &SystemState
  &self:SystemSpawnScript.SystemScriptSource.SystemScriptGraph.NiagaraNodeAssignment_0.SetVariables_#1
  &self:SystemSpawnScript.SystemScriptSource.SystemScriptGraph.NiagaraNodeAssignment_1.SetVariables_#2
 ]
}
emitters {Camera_Lens {
  emitterUpdateScript [
   &EmitterState(Life Cycle Mode=::NewEnumerator1, Loop Behavior=::NewEnumerator1, Loop Delay=0.05, Loop Duration Mode=::NewEnumerator1, UseLoopDelay=false)
   &SpawnBurst_Instantaneous(Spawn Count=1, Use Spawn Probability=false)
  ]
  particleSpawnScript [
   &InitializeParticle(
    Color=<rgba 1 1 1 3>
    Lifetime=0.3
    Mesh Scale=&Multiply_VectorByFloat(Vector=<1 1 1>, Float=0.37)
    Mesh Scale Mode=::NewEnumerator4
    UsePositionOffset=false
    Write Material Random=true
   )
   &InitialMeshOrientation(Rotation=<0 0 0.25>, Use Orientation Vector=true, Use Rotation Vector=true)
  ]
  particleUpdateScript [
   &ParticleState(Kill Particles When Lifetime Has Elapsed=false, Let Infinitely Lived Particles Die When Emitter Deactivates=true)
   &ScaleMeshSize(Scale Factor=&Multiply_VectorByFloat(
     Vector=<1 1 1>
     Float=&FloatFromCurve(FloatCurve=DataInterface() {propertyValues {Curve {keys [{tangentMode=RCTM_User, time=1, value=1.2}]}}})
    ))
   &DynamicMaterialParameters(
    Index 0 Param 2=&FloatFromCurve(FloatCurve=DataInterface() {propertyValues {Curve {keys [<at 0 1>, {arriveTangent=1, interpMode=RCIM_Cubic, leaveTangent=1, time=0.2566, value=1.6894}, <at 1 2>]}}})
    Index 0 Param 3=1
    Index 0 Param 4=1
    Module.Param0WriteEnabled=true
    Module.Param1WriteEnabled=true
    Module.Param2WriteEnabled=true
    Module.Param3WriteEnabled=true
   )
   &ScaleColor(
    Scale Alpha=&Multiply_Float(
     A=&Multiply_Float(
      A=$System.Flicker
      B=&FNTimeOfDayScalar(Night Scale=0.2, Morning Scale=0.5, Day Scale=0.5, Evening Scale=0.5)
     )
     B=&FloatFromCurve(
      FloatCurve=DataInterface() {propertyValues {Curve {keys [
          {arriveTangent=4.1785, interpMode=RCIM_Cubic, leaveTangent=4.1785, tangentMode=RCTM_User, time=0, value=0}
          {arriveTangent=1.1179, interpMode=RCIM_Cubic, leaveTangent=1.1179, tangentMode=RCTM_User, time=0.2042, value=0.6228}
          {arriveTangent=0.0224, interpMode=RCIM_Cubic, leaveTangent=0.0224, tangentMode=RCTM_User, time=1, value=1}
         ]}}}
      CurveIndex=$Particles.Age
     )
    )
    Scale RGB=&MakeVectorFromLinearColorRGB(Color=&ColorFromCurve(ColorCurve=DataInterface() {propertyValues {
        AlphaCurve {keys []}
        BlueCurve {keys [
          <at 0.0115 0.2615>
          {arriveTangent=1.1173, interpMode=RCIM_Cubic, leaveTangent=1.1173, time=0.3418, value=1}
          {arriveTangent=2.0008, interpMode=RCIM_Cubic, leaveTangent=2.0008, time=0.6725, value=1}
          <at 0.8416 2>
         ]}
        GreenCurve {keys [
          <at 0.0115 1>
          {arriveTangent=-0.707, interpMode=RCIM_Cubic, leaveTangent=-0.707, time=0.3418, value=1}
          {arriveTangent=-0.0094, interpMode=RCIM_Cubic, leaveTangent=-0.0094, time=0.6725, value=0.5327}
          <at 0.8416 0.9953>
         ]}
        RedCurve {keys [
          <at 0.0115 0>
          <at 0.3418 1 RCIM_Cubic>
          {arriveTangent=-2.0008, interpMode=RCIM_Cubic, leaveTangent=-2.0008, time=0.6725, value=0}
          <at 0.8416 0>
         ]}
       }}))
    ScaleA=true
    ScaleRGB=false
   )
  ]
  rendererClasses [&NiagaraMeshRendererProperties]
  renderers [{rendererClass=&NiagaraMeshRendererProperties, rendererIndex=0}]
 }}
summary {
 systemName=NS_NanaimoBar_Camera_Lens
 emitters [{emitterName=Camera_Lens, rendererClasses [&NiagaraMeshRendererProperties]}]
 userVariables [User.Overload {type=NiagaraFloat, defaultValue=0, description=""}]
}
