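# Niagara charge "film" effect for NanaimoBar that fades in a mesh, despite its emitter being named SimpleSpriteBurst.
# An instantaneous burst spawns mesh particles with a local position offset, slight scale stretch and camera offset, and they are not killed when their lifetime elapses.
# A curve indexed by emitter age ramps alpha up after a short delay.
# Gauntlet carapace-bone sampling and initial mesh orientation modules are present but disabled.
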NiagaraSystem /NanaimoBar/Assets/Fx/NS_NanaimoBar_Charge_Film
dependencies {
 usedDataInterfaces [&NiagaraDataInterfaceSkeletalMesh, &NiagaraDataInterfaceCurve]
 usedDynamicInputs []
 usedModules [&SystemState]
}
emitters {SimpleSpriteBurst {
  emitterUpdateScript [
   &EmitterState(Life Cycle Mode=::NewEnumerator1, Loop Behavior=::NewEnumerator1, Loop Duration=1, UseLoopDelay=false)
   &SpawnBurst_Instantaneous(Use Spawn Probability=false, UseLoopCountLimit=false)
  ]
  particleSpawnScript [
   &InitializeParticle(
    Lifetime=2
    Mesh Scale=<1 1.1 1>
    Mesh Scale Mode=::NewEnumerator4
    Position Offset=<-18.9 0.1 -14.8>
    Position Offset Coordinate Space=::Local
    Sprite Size Mode=::NewEnumerator3
    Uniform Sprite Size=50
    UsePositionOffset=true
    Write Material Random=true
   )
   &InitialMeshOrientation(Mesh Orientation Mode=::NewEnumerator1, Rotation=<0 0 -0.125>, Rotation Coordinate Space=::NewEnumerator2, Use Rotation Vector=true) {enabled=false}
   &CameraOffset(Camera Offset Amount=1)
  ]
  particleUpdateScript [
   &ParticleState(Kill Particles When Lifetime Has Elapsed=false)
   &SkeletalMeshLocation(
    Bone Sampling Mode=::NewEnumerator4
    MaskSpawnGroup=false
    Sampled Position Offset Coordinate Space:ENiagaraSkelMeshTransforms=::NewEnumerator2
    Skeletal Mesh=DataInterface() {propertyValues {
      DefaultMesh=/NanaimoBar/Assets/Meshes/NanaimoBar_Gauntlet
      ExcludeBoneName=root
      OptionalComponentClass=null
      PreviewMesh=&NanaimoBar_Gauntlet
      SkinningMode=SkinOnTheFly
      SoftSourceActor=null
      SourceMode=Default
      UvSetIndex=0
      WholeMeshLOD=-1
      bExcludeBone=false
      bReadDeformedGeometry=true
      bRequireCurrentFrameData=true
      ComponentTags []
      FilteredBones [carapace_outer, carapace_inner]
      FilteredSockets []
      MeshUserParameter {parameter {name=None, typeDefHandle {registeredTypeIndex=1}, varData []}}
      SamplingRegions []
     }}
   ) {enabled=false}
   &ScaleColor(
    Scale Alpha=&FloatFromCurve(
     FloatCurve=DataInterface() {propertyValues {Curve {defaultValue=1.0113, keys [{tangentMode=RCTM_User, time=0.2, value=0}, {tangentMode=RCTM_User, time=1, value=1}]}}}
     CurveIndex=$Emitter.Age
    )
    ScaleA=true
    ScaleRGB=false
   )
   &SolveForcesAndVelocity(Clamp Velocity=false, Limit Acceleration=false)
  ]
  rendererClasses [&NiagaraMeshRendererProperties]
  renderers [{rendererClass=&NiagaraMeshRendererProperties, rendererIndex=0}]
 }}
summary {
 systemName=NS_NanaimoBar_Charge_Film
 emitters [{emitterName=SimpleSpriteBurst, rendererClasses [&NiagaraMeshRendererProperties]}]
 userVariables []
}
