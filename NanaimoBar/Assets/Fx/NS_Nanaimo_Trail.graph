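# Niagara trail effect for a moving source, layering dark purple energy meshes, ribbons and goo splashes.
# DarkEnergyWaves001 bursts three indexed meshes that grow with emitter age, while DarkSPline001 continuously spawns rotating, fading meshes.
# Flame_Trail draws two ribbon layers, and DirSplash spawns falling sprites per distance travelled, gated by User.EnableGoo.
# User.FXScale drives sizes and colour shifts throughout; the three LW_ emitters have no renderers.
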NiagaraSystem /NanaimoBar/Assets/Fx/NS_Nanaimo_Trail
dependencies {
 usedDataInterfaces [
  &NiagaraDataInterfaceCurve
  &NiagaraDataInterfaceArrayFloat3
  &NiagaraDataInterfaceVectorCurve
  &NiagaraDataInterfaceArrayInt32
 ]
 usedDynamicInputs [
  &Lerp_Vector
  &Multiply_Vector2D_ByFloat
  &Sine
  &Vector2DFromFloat
  &Multiply_LinearColorByFloat
  &LerpLinearColors
  &FNTimeOfDayScalar
 ]
 usedModules [
  &self:SystemSpawnScript.SystemScriptSource.SystemScriptGraph.NiagaraNodeAssignment_0.SetVariables_#1
  &SystemState
  &self:SystemSpawnScript.SystemScriptSource.SystemScriptGraph.NiagaraNodeAssignment_1.SetVariables_#2
 ]
}
emitters {
 DarkEnergyWaves001 {
  emitterSpawnScript [/SetVariables(
    Emitter.DarkColor:LinearColor=/Niagara/DynamicInputs/Lerp/SmoothLerpOverTime_Color(
     Smooth Value:LinearColor=<rgba 0.0653 0 0.2448 1>
     Target Value:LinearColor=<rgba 37.3472 0 55 1>
     Convergence Rate:NiagaraFloat=&Multiply_Float(A=$User.FXScale, B=0.2)
    )
    Emitter.Linear Color:LinearColor=&SmoothLerpOverTime_Color(
     Smooth Value:LinearColor=<rgba 0.1333 0 0.5 1>
     Target Value:LinearColor=<rgba 6.7904 0 10 1>
     Convergence Rate:NiagaraFloat=&Multiply_Float(A=$User.FXScale, B=0.2)
    )
    Emitter.RandomisedColour:LinearColor=&RandomRangeLinearColor(Minimum=<rgba 0.1875 0.041 0.1846 1>, Maximum=<rgba 0.4471 0.518 1 1>)
    Emitter.TipBrightness:NiagaraFloat=&Lerp_Float(A=0, B=5, Alpha=&Multiply_Float(A=$User.FXScale, B=0.15))
   ) {isSetParameters=true}]
  emitterUpdateScript [
   &EmitterState(Inactive Response=::NewEnumerator1, Life Cycle Mode=::NewEnumerator1, UseLoopDelay=false)
   &SpawnBurst_Instantaneous(Spawn Count=3, Spawn Time=0.01, Use Spawn Probability=false, UseLoopCountLimit=false)
  ]
  particleSpawnScript [
   &InitializeParticle(
    Color=$Emitter.RandomisedColour
    Lifetime=$Emitter.CurrentLoopDuration
    Mesh Scale=&Multiply_Vector(
     A=&SelectVectorFromArray(
      Direct Array Index=$Engine.ExecIndex
      Vector Selection Array=DataInterface() {propertyValues {FloatData [<5 2.5 2.5>, <3 2 2>, <5 2.5 2.5>]}}
     )
     B=&RandomRangeVector(Minimum=<0.65 1 1>, Maximum=<1 1 1>)
    )
    Mesh Scale Mode=::NewEnumerator4
    UsePositionOffset=false
    Write Material Random=true
   )
   /SetVariables(Particles.CoreColor:LinearColor=&SmoothLerpOverTime_Color(
     Smooth Value:LinearColor=<rgba 0 0 0 0.5>
     Target Value:LinearColor=<rgba 0 0 0 0>
     Convergence Rate:NiagaraFloat=&Multiply_Float(A=$User.FXScale, B=0.15)
    )) {isSetParameters=true}
  ]
  particleUpdateScript [
   &ParticleState()
   &ScaleMeshSize(Scale Factor=&Multiply_Vector(
     A=&VectorFromCurve(
      VectorCurve=DataInterface() {propertyValues {
        XCurve {keys [
          {arriveTangent=2.7836, interpMode=RCIM_Cubic, leaveTangent=2.7836, tangentMode=RCTM_User, time=0.018, value=0.1}
          {arriveTangent=17.6272, tangentMode=RCTM_User, time=0.14, value=1}
         ]}
        YCurve {keys [<at 0 1>,<at 1 1>]}
        ZCurve {keys [<at 0 1>,<at 1 1>]}
       }}
      CurveIndex=$Emitter.Age
     )
     B=&VectorFromFloat(Value=&MakeCustomFloatFromBool(
       Bool=&SetBoolByFloatComparison(A=4, B=$User.FXScale)
       True Float=&Multiply_Float(
        A=&FloatFromCurve(
         FloatCurve=DataInterface() {propertyValues {Curve {defaultValue=1, keys [{tangentMode=RCTM_User, time=0.01, value=0.1}, {tangentMode=RCTM_User, time=0.07, value=1}]}}}
         CurveIndex=$Emitter.Age
        )
        B=$User.FXScale
       )
       False Float=&Multiply_Float(
        A=&FloatFromCurve(
         FloatCurve=DataInterface() {propertyValues {Curve {defaultValue=1, keys [<at 0 0 RCIM_Cubic>, {tangentMode=RCTM_User, time=0.07, value=1}]}}}
         CurveIndex=$Emitter.Age
        )
        B=$User.FXScale
       )
      ))
    ))
   /SetVariables(
    Particles.DarkEnergyScale:Vector3f=&Multiply_Vector(A=$Particles.Scale, B=<0.7 0.8 0.8>)
    Particles.VisibilityTag=&SelectIntFromArray(Direct Array Index=$Engine.ExecIndex, Int Selection Array=DataInterface() {propertyValues {IntData [0,0,1]}})
   ) {isSetParameters=true}
   &SolveForcesAndVelocity(Clamp Velocity=false, Limit Acceleration=false)
  ]
  rendererClasses [&NiagaraMeshRendererProperties]
  renderers [
   {rendererClass=&NiagaraMeshRendererProperties, rendererIndex=0}
   {rendererClass=&NiagaraMeshRendererProperties, rendererIndex=1}
  ]
 }
 DarkSPline001 {
  emitterUpdateScript [
   &EmitterState()
   &SpawnBurst_Instantaneous(Spawn Count=2, Spawn Time=0.01, Use Spawn Probability=false, UseLoopCountLimit=false)
   &SpawnRate(SpawnRate=10, Use Spawn Probability=false)
  ]
  particleSpawnScript [
   &InitializeParticle(
    Color Maximum=<rgba 0.0307 0.0206 0.1042 0.7>
    Color Minimum=<rgba 0.0377 0.0104 0.0417 0.5>
    Color Mode=::NewEnumerator2
    Lifetime Max=0.75
    Lifetime Min=0.45
    Lifetime Mode=::NewEnumerator1
    Mass Mode=::NewEnumerator3
    Mesh Scale=&Multiply_Vector(
     A=&RandomRangeVector(Minimum=<7 7 8>, Maximum=<9 9 11>)
     B=&MakeVector(
      X=$User.FXScale
      Y=&Multiply_Float(A=$User.FXScale, B=1.5)
      Z=&Multiply_Float(A=$User.FXScale, B=1.5)
     )
    )
    Mesh Scale Mode=::NewEnumerator4
    Sprite Rotation Mode=::NewEnumerator2
    Sprite Size Mode=::NewEnumerator1
    Uniform Sprite Size Max=12
    Uniform Sprite Size Min=6
    UsePositionOffset=false
    Write Material Random=true
   )
   &InitialMeshOrientation(
    Mesh Orientation Mode=::NewEnumerator7
    Rotation=&RandomRangeVector(Minimum=<0 0 0>, Maximum=<1 0 0>)
    Use Rotation Vector=true
   )
  ]
  particleUpdateScript [
   &ParticleState()
   &UpdateMeshOrientation(Rotation Rate=&RandomRangeFloat(Minimum=-0.4, Maximum=-1)) {module=UpdateMeshOrientation001}
   &ScaleMeshSize(Scale Factor=&VectorFromCurve(
     VectorCurve=DataInterface() {propertyValues {
       XCurve {keys [
         {arriveTangent=1.9584, interpMode=RCIM_Cubic, leaveTangent=1.9584, tangentMode=RCTM_User, time=0, value=0.2}
         {arriveTangent=1.1179, interpMode=RCIM_Cubic, leaveTangent=1.1179, tangentMode=RCTM_User, time=0.3, value=0.7}
         {arriveTangent=0.0224, interpMode=RCIM_Cubic, leaveTangent=0.0224, tangentMode=RCTM_User, time=1, value=1}
        ]}
       YCurve {keys [
         {arriveTangent=1.9584, interpMode=RCIM_Cubic, leaveTangent=1.9584, tangentMode=RCTM_User, time=0, value=0.2}
         {arriveTangent=1.1179, interpMode=RCIM_Cubic, leaveTangent=1.1179, tangentMode=RCTM_User, time=0.3, value=0.7}
         {arriveTangent=0.0224, interpMode=RCIM_Cubic, leaveTangent=0.0224, tangentMode=RCTM_User, time=1, value=1}
        ]}
       ZCurve {keys [
         {arriveTangent=1.8004, interpMode=RCIM_Cubic, leaveTangent=1.8004, tangentMode=RCTM_User, time=0, value=0.2}
         {arriveTangent=0.0224, interpMode=RCIM_Cubic, leaveTangent=0.0224, tangentMode=RCTM_User, time=1, value=1}
        ]}
      }}
     Scale Curve=&VectorFromFloat(Value=&FloatFromCurve(
       FloatCurve=DataInterface() {propertyValues {Curve {defaultValue=1, keys [{tangentMode=RCTM_User, time=0, value=0}, {tangentMode=RCTM_User, time=0.5, value=1}]}}}
       CurveIndex=$Emitter.Age
       Scale Curve=0.5
      ))
    ))
   &ScaleColor(
    Scale Alpha=&FloatFromCurve(
     FloatCurve=DataInterface() {propertyValues {Curve {keys [
         {arriveTangent=2.8016, interpMode=RCIM_Cubic, leaveTangent=2.8016, tangentMode=RCTM_User, time=0, value=0.1}
         {interpMode=RCIM_Cubic, tangentMode=RCTM_User, time=0.5, value=1}
         {tangentMode=RCTM_User, time=1, value=0}
        ]}}}
     Scale Curve=5
    )
    ScaleA=true
    ScaleRGB=true
   )
   &DynamicMaterialParameters(
    Index 3 Param 1=&FloatFromCurve(
     FloatCurve=DataInterface() {propertyValues {Curve {keys [{tangentMode=RCTM_User, time=0, value=0.5}, {tangentMode=RCTM_User, time=1, value=1}]}}}
     Scale Curve=&Multiply_Float(A=$User.FXScale, B=&RandomRangeFloat(Minimum=2, Maximum=2.5))
    )
    Index 3 Param 3=&RandomRangeFloat(Minimum=0, Maximum=1)
    Module.Index3Param0Write=true
    Module.Index3Param1Write=false
    Module.Index3Param2Write=true
    Module.Index3Param3Write=false
    Write Parameter Index 0=false
    Write Parameter Index 3=true
   )
   &SolveForcesAndVelocity(Clamp Velocity=false, Limit Acceleration=false)
  ]
  rendererClasses [&NiagaraMeshRendererProperties]
  renderers [{rendererClass=&NiagaraMeshRendererProperties, rendererIndex=0}]
 }
 DirSplash {
  emitterUpdateScript [
   &EmitterState()
   &SpawnPerUnit(
    Module.Use Max Movement Threshold=false
    Module.Use Movement Tolerance=false
    Spawn Probability:NiagaraFloat=&MakeCustomFloatFromBool(Bool=$User.EnableGoo, True Float=1, False Float=0)
    Spawn Spacing=90
    Use Spawn Probability=true
   )
   /SetVariables(Emitter.FresnelColor:LinearColor=<rgba 0.0552 0.0075 0.1198 1>, Emitter.TileColor:LinearColor=<rgba 0.0477 0.0061 0.0833 1>) {isSetParameters=true}
  ]
  particleSpawnScript [&InitializeParticle(
    Color Maximum=<rgba 0.0093 0.0015 0.0104 1>
    Color Minimum=<rgba 0.0032 0 0.0208 1>
    Color Mode=::NewEnumerator2
    Lifetime=&RandomRangeFloat(Minimum=0.6, Maximum=0.35)
    Sprite Rotation Mode=::NewEnumerator2
    Sprite Size Max=<22 22>
    Sprite Size Min=<17 15>
    Sprite Size Mode=::NewEnumerator2
    Sprite UV Mode=::NewEnumerator3
    UsePositionOffset=false
    Write Material Random=true
   )]
  particleUpdateScript [
   &ParticleState()
   &DynamicMaterialParameters(
    Index 0 Param 1=&RandomRangeFloat(Minimum=0, Maximum=1)
    Index 0 Param 4=&FloatFromCurve(FloatCurve=DataInterface() {propertyValues {Curve {defaultValue=1, keys [{tangentMode=RCTM_User, time=0.2, value=-1}, {tangentMode=RCTM_User, time=1, value=2}]}}})
    Module.Param0WriteEnabled=true
    Module.Param0WriteEnabled1=true
    Module.Param1WriteEnabled=true
    Module.Param1WriteEnabled1=true
    Module.Param2WriteEnabled=true
    Module.Param2WriteEnabled1=true
    Module.Param3WriteEnabled=true
    Module.Param3WriteEnabled1=true
    Write Parameter Index 1=true
   )
   &ScaleSpriteSize(
    Uniform Curve Scale=&Multiply_Float(
     A=&Lerp_Float(A=3, B=5, Alpha=&Multiply_Float(A=0.15, B=$User.FXScale))
     B=&FloatFromCurve(
      FloatCurve=DataInterface() {propertyValues {Curve {defaultValue=1, keys [<at 0.2 0 RCIM_Cubic>, {tangentMode=RCTM_User, time=1, value=1}]}}}
      CurveIndex=$Emitter.Age
      Scale Curve=&RandomRangeFloat(Minimum=0.7, Maximum=1)
     )
    )
    Uniform Curve Sprite Scale=DataInterface() {propertyValues {Curve {keys [
        {arriveTangent=8.5553, interpMode=RCIM_Cubic, leaveTangent=8.5553, tangentMode=RCTM_User, time=0, value=1.2}
        {arriveTangent=4.3444, interpMode=RCIM_Cubic, leaveTangent=4.3444, tangentMode=RCTM_User, time=0.199, value=2.3279}
        {arriveTangent=3.5267, interpMode=RCIM_Cubic, leaveTangent=3.5267, tangentMode=RCTM_User, time=0.6128, value=2.9347}
        {arriveTangent=0.7114, tangentMode=RCTM_User, time=1.0095, value=3.9153}
       ]}}}
   )
   &GravityForce(Gravity=&RandomRangeVector(Minimum=<0 0 -300>, Maximum=<0 0 0>))
   &InheritSourceMovement(Applied Position Scale=<0.1 0.1 0.1>, ApplyForce=false, ApplyPosition=true, ApplyPositionFalloff=false)
   &SolveForcesAndVelocity(Clamp Velocity=false, Limit Acceleration=false)
   &ScaleColor(
    Scale Alpha=&FloatFromCurve(FloatCurve=DataInterface() {propertyValues {Curve {keys [<at 0.3 0>,<at 0.3766 0.6265>,<at 0.5 1>,<at 1 0>]}}})
    Scale RGB=<0.1 0.1 0.1>
    ScaleA=true
    ScaleRGB=false
   )
  ]
  rendererClasses [&NiagaraSpriteRendererProperties]
  renderers [{rendererClass=&NiagaraSpriteRendererProperties}]
 }
 Flame_Trail {
  emitterSpawnScript [/SetVariables(Emitter.RandomisedColour:LinearColor=&RandomRangeLinearColor(Minimum=<rgba 0.1984 0.1701 0.4844 1>, Maximum=<rgba 0.9162 0.2831 1 1>)) {isSetParameters=true}]
  emitterUpdateScript [
   &EmitterState()
   &SpawnRate(SpawnRate=30, Use Spawn Probability=false)
  ]
  particleSpawnScript [
   &InitializeParticle(
    Color=$Emitter.RandomisedColour
    Lifetime=0.57
    Position Offset=<200 0 0>
    Ribbon Width=&Multiply_Float(A=12, B=$User.FXScale)
    Ribbon Width Mode=::NewEnumerator1
    UsePositionOffset=false
    Write Material Random=true
   )
   /SetVariables(Particles.R2_RibbonWidth:NiagaraFloat=&Multiply_Float(A=18, B=$User.FXScale)) {isSetParameters=true}
  ]
  particleUpdateScript [
   &ParticleState()
   &DynamicMaterialParameters(
    Index 0 Param 2=&Multiply_Float(A=$Emitter.Age, B=7)
    Index 0 Param 4=&FloatFromCurve(FloatCurve=DataInterface() {propertyValues {Curve {keys [{arriveTangent=-0.9325, time=0.4, value=1}, {arriveTangent=-1.7559, tangentMode=RCTM_User, time=1, value=0}]}}})
    Module.Index0Param0Write=true
    Module.Index0Param1Write=true
    Module.Index0Param2Write=true
    Module.Index0Param3Write=true
   )
   &ScaleColor(
    Scale Alpha=&FloatFromCurve(FloatCurve=DataInterface() {propertyValues {Curve {keys [<at 0 1>, {arriveTangent=-5.7009, time=0.4, value=1}, <at 1 0>]}}}, Scale Curve=2)
    ScaleA=true
    ScaleRGB=false
   )
   /SetVariables(Particles.Color2=<rgba 1 1 1 0.7>) {isSetParameters=true}
  ]
  rendererClasses [&NiagaraRibbonRendererProperties]
  renderers [
   {rendererClass=&NiagaraRibbonRendererProperties}
   {rendererClass=&NiagaraRibbonRendererProperties, rendererIndex=1}
  ]
 }
 LW_Core {rendererClasses [], renderers []}
 LW_Core001 {rendererClasses [], renderers []}
 LW_Plasma_Center_Energy {rendererClasses [], renderers []}
}
summary {
 systemName=NS_Nanaimo_Trail
 emitters [
  {emitterName=Flame_Trail, rendererClasses [&NiagaraRibbonRendererProperties]}
  {emitterName=LW_Plasma_Center_Energy, rendererClasses []}
  {emitterName=LW_Core, rendererClasses []}
  {emitterName=DarkEnergyWaves001, rendererClasses [&NiagaraMeshRendererProperties]}
  {emitterName=DarkSPline001, rendererClasses [&NiagaraMeshRendererProperties]}
  {emitterName=DirSplash, rendererClasses [&NiagaraSpriteRendererProperties]}
  {emitterName=LW_Core001, rendererClasses []}
 ]
 userVariables [
  User.EnableGoo {type=NiagaraBool, defaultValue=true}
  User.FXLerp {type=NiagaraFloat, defaultValue=1, description=""}
  User.FXScale {type=NiagaraFloat, defaultValue=15, description=""}
 ]
}
