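# Niagara charge-up effect for a projectile, building a violet glow that peaks in a flash as charge completes.
# System.ChargeUpTime indexes curves that grow a pulsing sprite, mesh and light, then trigger brief shockwave and hot-debris spawn windows near the peak.
# A wind cone instead scales, tints and fades by User.Charge Percent, with a threshold kill.
# Time-of-day inputs adjust shockwave and mesh color; the flash decal emitter is disabled.
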NiagaraSystem /NanaimoBar/Assets/Fx/NS_Projectile_ChargeUp
dependencies {
 usedDataInterfaces [
  &NiagaraDataInterfaceCurve
  &NiagaraDataInterfaceColorCurve
  &NiagaraDataInterfaceVectorCurve
  &NiagaraDataInterfaceSpriteRendererInfo
 ]
 usedDynamicInputs [&RandomRangeLinearColor, &Add_Float]
 usedModules [
  &self:SystemSpawnScript.SystemScriptSource.SystemScriptGraph.NiagaraNodeAssignment_1.SetVariables_#1
  &SystemState
  &Timeline
  &self:SystemSpawnScript.SystemScriptSource.SystemScriptGraph.NiagaraNodeAssignment_2.SetVariables_#2
  &self:SystemSpawnScript.SystemScriptSource.SystemScriptGraph.NiagaraNodeAssignment_0.SetVariables_#3
 ]
}
emitters {
 Bright_Flash_Decal {
  bEnabled=false
  emitterUpdateScript [
   &EmitterState(Inactive Response=::NewEnumerator1, Life Cycle Mode=::NewEnumerator1, Loop Delay=0.1, Loop Duration=6, UseLoopDelay=false)
   &SpawnRate(SpawnRate=20, Use Spawn Probability=false)
   &SpawnBurst_Instantaneous(Spawn Count=1, Spawn Time=0.1, Use Spawn Probability=false)
  ]
  particleSpawnScript [
   &PartitionParticles(Number Of Partitions=13)
   /Niagara/Modules/Spawn/Initialization/InitializeParticle(
    Color=<rgba 0.2848 0 0.9 0.9>
    Lifetime=&RandomRangeFloat(Minimum=0.7, Maximum=0.8)
    Mass=&RandomRangeFloat(Minimum=0.6, Maximum=1.2)
    Module.Write Color=true
    Module.Write Lifetime=true
    Module.Write Mass=false
    Module.Write Material Random=true
    Module.Write Position=false
    Module.Write Scale=false
    Module.Write SpriteRotation=true
    Module.Write SpriteSize=true
    Sprite Rotation=&RandomRangeFloat(Minimum=0, Maximum=360)
    Sprite Size=&Vector2DFromFloat(Value=&Multiply_Float(A=60, B=&RandomRangeFloat(Minimum=1, Maximum=1.2)))
   )
   &CameraOffset(Camera Offset Amount=40)
   &ShapeLocation(MaskSpawnGroup=false, Offset=<-1 0 0>, Offset Mode=::NewEnumerator0, Override Seed=false, Sphere Radius=10, UseOwnerScale=true)
   &AddVelocity(Velocity=<50 0 0>)
  ]
  particleUpdateScript [
   &ParticleState()
   &ScaleSpriteSize(
    Uniform Curve Index=$System.ChargeUpTime
    Uniform Curve Sprite Scale=DataInterface() {propertyValues {Curve {keys [<at 0 0>,<at 0.4842 0.3506>,<at 0.97 0.3506>,<at 1.05 0.8 RCIM_Constant>]}}}
   )
   &ScaleColor(
    Scale Alpha=&FloatFromCurve(FloatCurve=DataInterface() {propertyValues {Curve {keys [
         {arriveTangent=0.5269, interpMode=RCIM_Cubic, leaveTangent=0.5269, tangentMode=RCTM_User, time=0, value=1}
         {arriveTangent=-0.2619, interpMode=RCIM_Cubic, leaveTangent=-0.2619, tangentMode=RCTM_User, time=0.3, value=3}
         <at 1 0 RCIM_Cubic>
        ]}}})
    Scale RGB=<0 0 0>
    ScaleA=true
    ScaleRGB=false
   )
   &DynamicMaterialParameters(
    Index 0 Param 1=&FloatFromCurve(FloatCurve=DataInterface() {propertyValues {Curve {keys [<at 0 5 RCIM_Cubic>,<at 0.0296 1.5 RCIM_Cubic>,<at 1 5 RCIM_Cubic>]}}})
    Index 0 Param 2=1
    Index 0 Param 3=&FloatFromCurve(FloatCurve=DataInterface() {propertyValues {Curve {keys [<at 0 0.35>,<at 1 0>]}}})
    Index 0 Param 4=&RandomRangeFloat(Minimum=0, Maximum=1)
    Index 1 Param 1=200
    Index 1 Param 2=5
    Module.Index0Float0Write=true
    Module.Index0Float1Write=true
    Module.Index0Float2Write=true
    Module.Index0Float3Write=true
    Module.Index1Float0Write=true
    Module.Index1Float1Write=true
    Module.Index1Float2Write=false
    Module.Index1Float3Write=false
    Write Parameter Index 1=::NewEnumerator1
   )
   &SolveForcesAndVelocity(Clamp Velocity=false, Limit Acceleration=false)
  ]
  rendererClasses [&NiagaraSpriteRendererProperties]
  renderers [{rendererClass=&NiagaraSpriteRendererProperties}]
 }
 HotDebris001 {
  emitterUpdateScript [
   &EmitterState(Life Cycle Mode=::NewEnumerator1, Loop Delay=1.9, Loop Duration=0.2, UseLoopDelay=false)
   &SpawnBurst_Instantaneous(Spawn Count=40, Use Spawn Probability=false, UseLoopCountLimit=false) {enabled=false}
   &SpawnRate(
    SpawnRate=&FloatFromCurve(
     FloatCurve=DataInterface() {propertyValues {Curve {keys [<at 0 0 RCIM_Constant>,<at 1.1696 1 RCIM_Constant>,<at 1.2496 0 RCIM_Constant>]}}}
     CurveIndex=$System.ChargeUpTime
     Scale Curve=500
    )
    Use Spawn Probability=false
   ) {module=SpawnRate002}
  ]
  particleSpawnScript [
   &PartitionParticles(
    B:NiagaraFloat=&RandomRangeFloat(Minimum=0, Maximum=1.4)
    Partition Mode:ENiagara_PartitionMode_v2=::NewEnumerator4
   )
   /Niagara/Modules/Spawn/Initialization/V2/InitializeParticle(
    Color Maximum=<rgba 36.8266 0.4167 40 0.3>
    Color Minimum=<rgba 0.2299 0 5 0.2>
    Color Mode=::NewEnumerator2
    Lifetime=&UniformRangedFloat(Minimum=0.2, Maximum=0.5)
    Sprite Size Max=<2 2>
    Sprite Size Min=<1 1>
    Sprite Size Mode=::NewEnumerator2
    UsePositionOffset=false
    Write Material Random=true
   ) {module=InitializeParticle001}
   &SubUVAnimation(
    Evaluation Type=::NewEnumerator0
    "Sprite Renderer "=DataInterface() {propertyValues {SpriteRenderer=&self:HotDebris001_0.NiagaraSpriteRendererProperties_1}}
    SubUV Animation Mode:ENiagara_SubUVLookupModeV2=::NewEnumerator2
    UseEndFrame=false
    UseStartFrame=false
   )
   &ShapeLocation(MaskSpawnGroup=false, Override Seed=false, Sphere Radius=35, UseOwnerScale=true)
   &AddVelocity(
    Module.Use Falloff Distance=false
    Override Seed=false
    Velocity Mode=::NewEnumerator1
    Velocity Speed=&RandomRangeFloat(Minimum=150, Maximum=400)
   )
  ]
  particleUpdateScript [
   &Drag(Drag=1.5, Use Linear Drag=true, Use Rotational Drag=true)
   &SolveForcesAndVelocity(Clamp Velocity=false, Limit Acceleration=false)
   &ParticleState()
   &ScaleSpriteSizeBySpeed(Max Scale Factor=<0.7 60>, Min Scale Factor=<1.5 1>)
   &ScaleSpriteSize(Scale Factor=&Vector2DFromFloat(Value=&FloatFromCurve(FloatCurve=DataInterface() {propertyValues {Curve {keys [<at 0 0.5>,<at 0.1 0.9>,<at 1 0>]}}})))
   &ScaleColor(
    Scale RGB=&VectorFromFloat(Value=&FloatFromCurve(FloatCurve=DataInterface() {propertyValues {Curve {keys [
          {arriveTangent=-2.3795, interpMode=RCIM_Cubic, leaveTangent=-2.3795, tangentMode=RCTM_User, time=-0.0046, value=1.0113}
          <at 1.0011 0.002 RCIM_Cubic>
         ]}}}))
    ScaleA=false
    ScaleRGB=true
   )
  ]
  rendererClasses [&NiagaraSpriteRendererProperties]
  renderers [{rendererClass=&NiagaraSpriteRendererProperties}]
 }
 LW_Chunks {rendererClasses [], renderers []}
 LW_Radial_Plasma_Energy {rendererClasses [], renderers []}
 LW_Radial_Plasma_Energy001 {rendererClasses [], renderers []}
 LW_Radial_Plasma_Energy002 {bEnabled=false, rendererClasses [], renderers []}
 LW_back {rendererClasses [], renderers []}
 Light_and_Flare {
  emitterUpdateScript [
   &EmitterState(
    Inactive Response=::NewEnumerator1
    Life Cycle Mode=::NewEnumerator1
    Loop Behavior=::NewEnumerator1
    Loop Duration Mode=::NewEnumerator1
    UseLoopDelay=false
   )
   &SpawnBurst_Instantaneous(Use Spawn Probability=false, UseLoopCountLimit=false)
  ]
  particleSpawnScript [
   &InitializeParticle(
    Color=<rgba 0.8 0.8 0.8 0.8>
    Lifetime=20
    Mesh Scale Mode=::NewEnumerator3
    Mesh Uniform Scale=0.25
    Sprite Rotation Angle=2
    Sprite Rotation Mode=::NewEnumerator0
    Sprite Size Mode=::NewEnumerator3
    Uniform Sprite Size=1
    UsePositionOffset=false
    Write Material Random=true
   )
   &CameraOffset()
   &DynamicMaterialParameters(
    Index 0 Param 1=1
    Index 0 Param 2=1
    Index 0 Param 3=1
    Index 0 Param 4=1
    Index 1 Param 1=30
    Module.Index0Param0Write=true
    Module.Index0Param1Write=true
    Module.Index0Param2Write=true
    Module.Index0Param3Write=true
    Module.Index1Param0Write=true
    Module.Index1Param1Write=true
    Module.Index1Param2Write=true
    Module.Index1Param3Write=true
    Write Parameter Index 1=true
   )
  ]
  particleUpdateScript [
   &ParticleState(
    Kill Particles When Lifetime Has Elapsed=false
    Let Infinitely Lived Particles Die When Emitter Deactivates=true
    Loop Particles Lifetime=true
   )
   &ScaleMeshSize(Scale Factor=&Multiply_VectorByFloat(
     Vector=<1 1 1>
     Float=&FloatFromCurve(
      FloatCurve=DataInterface() {propertyValues {Curve {keys [
          <at 0.2345 0 RCIM_Cubic>
          {interpMode=RCIM_Cubic, tangentMode=RCTM_User, time=0.4172, value=0.3102}
          <at 0.9785 0.5>
          <at 1.0307 1.5>
          <at 1.2 0.8 RCIM_Constant>
         ]}}}
      CurveIndex=$System.ChargeUpTime
      Scale Curve=&Sine(Normalized Angle=$Engine.System.Age, Period=0.2, Scale=0.1, Bias=0.95)
     )
    ))
   &ScaleSpriteSize(
    Uniform Curve Index=$System.ChargeUpTime
    Uniform Curve Scale=&Sine(Normalized Angle=$Engine.System.Age, Period=0.2, Scale=35, Bias=665)
    Uniform Curve Sprite Scale=DataInterface() {propertyValues {Curve {keys [
        <at 0.1 0>
        <at 0.4 0.15 RCIM_Constant>
        <at 1.1 0.2882>
        {arriveTangent=-7.5789, interpMode=RCIM_Cubic, leaveTangent=-7.5789, tangentMode=RCTM_User, time=1.2537, value=1.319}
        <at 1.4 0.5 RCIM_Constant>
       ]}}}
   )
   &SpriteRotationRate(Rotation Rate=2)
   &ScaleColor(
    Scale Alpha=&FloatFromCurve(
     FloatCurve=DataInterface() {propertyValues {Curve {defaultValue=1, keys [{tangentMode=RCTM_User, time=0, value=0}, {tangentMode=RCTM_User, time=0.2, value=1}]}}}
     CurveIndex=$Emitter.Age
     Scale Curve=&FloatFromCurve(
      FloatCurve=DataInterface() {propertyValues {Curve {keys [<at 0.1 0>,<at 0.4 1 RCIM_Constant>,<at 1.0465 1>,<at 1.1367 2.2>,<at 1.2997 0.5 RCIM_Constant>]}}}
      CurveIndex=$System.ChargeUpTime
      Scale Curve=2
     )
    )
    Scale RGB=&MakeVectorFromLinearColorRGB(Color=<rgba 0.6289 0.0856 1 1>)
    ScaleA=true
    ScaleRGB=true
   )
   &FN_Light_Attributes(
    Color=&Multiply_LinearColorByFloat(
     Linear Color=&LerpLinearColors(
      StartColor=<rgba 0.4585 0.2434 4 1>
      EndColor=<rgba 0.4082 0.1357 4 1>
      LerpFactor=&RandomRangeFloat(Minimum=0.4, Maximum=0.6, Evaluation Type=::NewEnumerator1)
     )
     Float=&FloatFromCurve(
      FloatCurve=DataInterface() {propertyValues {Curve {defaultValue=1, keys [{tangentMode=RCTM_User, time=0, value=0}, {tangentMode=RCTM_User, time=0.5, value=1}]}}}
      CurveIndex=$Emitter.Age
     )
    )
    Create Color Attribute=true
    Create Exponent Attribute=true
    Create Radius Attribute=true
    Exponent=2
    Radius=&FloatFromCurve(
     FloatCurve=DataInterface() {propertyValues {Curve {keys [<at 0 0 RCIM_Constant>,<at 1.0049 0.45>,<at 1.1014 1.5353>,<at 1.1968 0.6824 RCIM_Constant>]}}}
     CurveIndex=$System.ChargeUpTime
     Scale Curve=70
    )
    Scale Color=&RandomRangeFloat(Minimum=3, Maximum=4, Evaluation Type=::NewEnumerator1)
   )
   /SetVariables(
    Particles.Color2=&ColorFromCurve(
     ColorCurve=DataInterface() {propertyValues {
       AlphaCurve {keys [<at 0.3266 0 RCIM_Cubic>,<at 1 1 RCIM_Cubic>]}
       BlueCurve {keys [<at 0 15 RCIM_Cubic>]}
       GreenCurve {keys [<at 0 0 RCIM_Cubic>]}
       RedCurve {keys [<at 0 2.8051 RCIM_Cubic>]}
      }}
     CurveIndex=$Emitter.Age
    )
    Particles.Mesh Color:LinearColor=&MakeLinearColorFromVectorAndFloat(
     "Vector (RGB)"=&MakeVectorFromLinearColorRGB(Color=&FNTimeOfDayColor(Night Color=<rgba 0.05 0 0.0248 1>, Morning Color=<rgba 0.03 0 0.05 1>, Day Color=<rgba 0.03 0 0.05 1>, Evening Color=<rgba 0.03 0 0.05 1>))
     "Float (Alpha)"=&FloatFromCurve(
      FloatCurve=DataInterface() {propertyValues {Curve {
         defaultValue=1
         keys [<at 0.5167 0 RCIM_Cubic>, {arriveTangent=2.4062, interpMode=RCIM_Cubic, leaveTangent=2.4062, tangentMode=RCTM_User, time=1.2, value=1}]
        }}}
      CurveIndex=$System.ChargeUpTime
      Scale Curve=1.1
     )
    )
    Particles.Mesh Pos:NiagaraPosition=&AddVectorToPosition(
     Position=$Engine.Emitter.SimulationPosition
     Vector=&MakeVector(
      X=&FloatFromCurve(
       FloatCurve=DataInterface() {propertyValues {Curve {keys [
           <at 0.25 1>
           {arriveTangent=17.0591, interpMode=RCIM_Cubic, leaveTangent=17.0591, tangentMode=RCTM_User, time=1.0356, value=0.7746}
           <at 1.2363 3.5511>
           <at 1.3368 0.7532>
          ]}}}
       CurveIndex=$System.ChargeUpTime
       Scale Curve=5
      )
      Y=0
      Z=0
     )
    )
    Particles.MeshGlow_Color:LinearColor=&ColorFromCurve(
     ColorCurve=DataInterface() {propertyValues {
       AlphaCurve {keys [<at 0.5 0 RCIM_Cubic>, {arriveTangent=1.0586, interpMode=RCIM_Cubic, leaveTangent=1.0586, tangentMode=RCTM_User, time=1, value=0.1}]}
       BlueCurve {keys [<at 0 2.0766 RCIM_Cubic>]}
       GreenCurve {keys [<at 0 0.2031 RCIM_Cubic>]}
       RedCurve {keys [<at 0 3 RCIM_Cubic>]}
      }}
     CurveIndex=$Emitter.Age
    )
    Particles.Sprite2:Vector2f=&Multiply_Vector2D_ByFloat(
     Float=&FloatFromCurve(
      FloatCurve=DataInterface() {propertyValues {Curve {keys [
          <at 0.1 0>
          <at 0.1421 0.1091>
          <at 0.1697 0.0441>
          <at 0.2052 0.1222>
          <at 0.2368 0.018>
          <at 0.2723 0.1091>
          <at 0.2998 0.0132>
          <at 0.3335 0.1688>
          <at 0.361 0.0562>
          <at 0.4019 0.1882>
          <at 0.4262 0.1007>
          <at 0.4695 0.1612>
          <at 0.503 0.091>
          <at 0.5326 0.2169>
          <at 0.578 0.1091>
          <at 0.6175 0.2378>
          <at 0.6471 0.0961>
          <at 0.6844 0.2763>
          <at 0.7141 0.1222>
          <at 0.7378 0.3032>
          <at 0.7575 0.1222>
          <at 0.7989 0.3196>
          <at 0.8187 0.0831>
          <at 0.8403 0.345>
          <at 0.8719 0.1091>
          <at 0.9054 0.405>
          <at 0.9252 0.1612>
          <at 0.9607 0.3044>
          <at 0.9922 0.3504>
          {arriveTangent=3.1943, interpMode=RCIM_Cubic, leaveTangent=3.1943, time=1.0477, value=0.2882}
          <at 1.2895 1.3>
          <at 1.3733 0.6 RCIM_Constant>
         ]}}}
      CurveIndex=$System.ChargeUpTime
      Scale Curve=&Sine(Normalized Angle=$Engine.System.Age, Period=0.2, Scale=35, Bias=500)
     )
     Vector2D=<0.1 0.1>
    )
   ) {isSetParameters=true}
  ]
  rendererClasses [
   &NiagaraSpriteRendererProperties
   &NiagaraLightRendererProperties
   &NiagaraMeshRendererProperties
  ]
  renderers [
   {rendererClass=&NiagaraSpriteRendererProperties}
   {rendererClass=&NiagaraLightRendererProperties, rendererIndex=1}
   {rendererClass=&NiagaraMeshRendererProperties, rendererIndex=2}
   {rendererClass=&NiagaraSpriteRendererProperties, rendererIndex=3}
  ]
 }
 Shockwave01 {
  emitterUpdateScript [
   &EmitterState(Inactive Response=::NewEnumerator1, Life Cycle Mode=::NewEnumerator1, Loop Behavior=::NewEnumerator1, Loop Duration=3, UseLoopDelay=false)
   &SpawnRate(
    SpawnRate=&FloatFromCurve(
     FloatCurve=DataInterface() {propertyValues {Curve {keys [<at 0 0 RCIM_Constant>,<at 0.95 1.0336 RCIM_Constant>,<at 1 0 RCIM_Constant>]}}}
     CurveIndex=$System.ChargeUpTime
     Scale Curve=100
    )
    Use Spawn Probability=false
   )
   &SpawnBurst_Instantaneous(Use Spawn Probability=false, UseLoopCountLimit=false) {enabled=false}
   &SpawnBurst_Instantaneous(Use Spawn Probability=false, UseLoopCountLimit=false) {enabled=false, module=SpawnBurst_Instantaneous001}
  ]
  particleSpawnScript [
   &InitializeParticle(
    Color=<rgba 1.2781 1.2936 2.5 0.5>
    Lifetime=0.6
    Mesh Scale Mode=::NewEnumerator3
    Mesh Uniform Scale=&Lerp_Float(A=0.15, B=1, Alpha=$Emitter.NormalizedLoopAge)
    UsePositionOffset=false
    Write Material Random=true
   )
   &InitialMeshOrientation(Orientation Axis=<0 0 1>, Orientation Vector=<0 0 1>, Rotation=<0 0.25 0>, Use Orientation Vector=true, Use Rotation Vector=true)
   &DynamicMaterialParameters(
    Index 0 Param 2=1
    Index 0 Param 3=1
    Index 0 Param 4=1
    Module.Param0WriteEnabled=false
    Module.Param1WriteEnabled=true
    Module.Param2WriteEnabled=true
    Module.Param3WriteEnabled=true
   )
   &InitialMeshOrientation(Orientation Axis=<0 0 1>, Orientation Vector=<0 0 1>, Rotation=<-0.8 0 0>, Use Orientation Vector=true, Use Rotation Vector=true) {module=InitialMeshOrientation001}
   &SystemLocation(MaskSpawnGroup=false, Offset=<12 -2 -2>)
  ]
  particleUpdateScript [
   &ParticleState()
   &ScaleMeshSize(Scale Factor=&VectorFromFloat(Value=&FloatFromCurve(FloatCurve=DataInterface() {propertyValues {Curve {keys [<at 0.1 0.5>,<at 0.2792 0.9887>,<at 1 1.3759>]}}})))
   &DynamicMaterialParameters(
    Index 0 Param 1=&FloatFromCurve(FloatCurve=DataInterface() {propertyValues {Curve {keys [{arriveTangent=3.0542, interpMode=RCIM_Cubic, leaveTangent=3.0542, tangentMode=RCTM_User, time=0, value=-1}, <at 1 0.3>]}}})
    Index 1 Param 1=&FloatFromCurve(FloatCurve=DataInterface() {propertyValues {Curve {keys [{arriveTangent=-3.2718, interpMode=RCIM_Cubic, leaveTangent=-3.2718, tangentMode=RCTM_User, time=0, value=1}, <at 1 -0.4>]}}})
    Index 1 Param 2=&FloatFromCurve(FloatCurve=DataInterface() {propertyValues {Curve {keys [{arriveTangent=-0.5818, interpMode=RCIM_Cubic, leaveTangent=-0.5818, tangentMode=RCTM_User, time=0, value=1}]}}})
    Index 1 Param 3=&FloatFromCurve(FloatCurve=DataInterface() {propertyValues {Curve {keys [<at 0 0>,<at 0.1 1>]}}})
    Module.Param0WriteEnabled=true
    Module.Param0WriteEnabled1=true
    Module.Param1WriteEnabled=false
    Module.Param1WriteEnabled1=true
    Module.Param2WriteEnabled=false
    Module.Param2WriteEnabled1=true
    Module.Param3WriteEnabled=false
    Module.Param3WriteEnabled1=true
    Write Parameter Index 1=true
   ) {module=DynamicMaterialParameters001}
   &ScaleColor(
    Scale Alpha=&FloatFromCurve(
     FloatCurve=DataInterface() {propertyValues {Curve {keys [<at 0 0>,<at 0.1 1>,<at 1 0>]}}}
     Scale Curve=&FloatFromCurve(FloatCurve=DataInterface() {propertyValues {Curve {keys [<at 0 0.5>,<at 0.2 1>,<at 1 0>]}}}, Scale Curve=0.35)
    )
    Scale RGB=&VectorFromFloat(Value=&FloatFromCurve(FloatCurve=DataInterface() {propertyValues {Curve {keys [<at 0 0>,<at 0.3 1>,<at 1 1>]}}}, Scale Curve=0.1))
    ScaleA=true
    ScaleRGB=true
   )
   &ScaleColor(
    Scale Alpha=&FNTimeOfDayScalar(Night Scale=0.7, Morning Scale=0.3, Evening Scale=0.5)
    ScaleA=true
    ScaleRGB=false
   ) {module=ScaleColor001}
   &MeshRotationForce(Rotation=&MakeVector(
     X=&FloatFromCurve(FloatCurve=DataInterface() {propertyValues {Curve {keys [<at 0 1>,<at 1 0>]}}})
     Y=0
     Z=0
    ))
   &SolveForcesAndVelocity(Clamp Velocity=false, Limit Acceleration=false)
   &SolveRotationalForcesAndVelocity()
   /SetVariables(Particles.Shockwave2Scale:Vector3f=&Multiply_Vector(
     A=$Particles.Initial.Scale
     B=&VectorFromFloat(Value=&FloatFromCurve(FloatCurve=DataInterface() {propertyValues {Curve {keys [<at 0 0>,<at 0.05 1>,<at 1 1>]}}}, Scale Curve=1.5))
    )) {isSetParameters=true}
  ]
  rendererClasses [&NiagaraMeshRendererProperties]
  renderers [
   {rendererClass=&NiagaraMeshRendererProperties, rendererIndex=0}
   {rendererClass=&NiagaraMeshRendererProperties, rendererIndex=1}
  ]
 }
 WindCone {
  emitterUpdateScript [
   &EmitterState()
   &SpawnBurst_Instantaneous(Spawn Time=0.03, Use Spawn Probability=false, UseLoopCountLimit=false)
  ]
  particleSpawnScript [
   &InitializeParticle(
    Color=<rgba 3.5 0.031 1.9239 1>
    Lifetime=1.5
    Mesh Scale=&Multiply_VectorByFloat(Vector=<0.5 1.15 1.15>, Float=0.25)
    Mesh Scale Mode=::NewEnumerator4
    Sprite Size Mode=::NewEnumerator1
    Uniform Sprite Size Max=150
    Uniform Sprite Size Min=100
    UsePositionOffset=false
    Write Material Random=true
   ) {module=InitializeParticle001}
   &DynamicMaterialParameters(
    Index 0 Param 1=1
    Index 0 Param 2=0.6
    Index 0 Param 3=2
    Module.Param0WriteEnabled=true
    Module.Param1WriteEnabled=true
    Module.Param2WriteEnabled=true
    Module.Param3WriteEnabled=false
   )
   &InitialMeshOrientation(Mesh Axis To Orient=<-1 0 0>, Mesh Orientation Mode=::NewEnumerator2, Use Rotation Vector=false)
   &CameraOffset(Camera Offset Amount=20)
  ]
  particleUpdateScript [
   &ParticleState(Kill Particles When Lifetime Has Elapsed=false)
   &ScaleMeshSize(Scale Factor=&VectorFromCurve(
     VectorCurve=DataInterface() {propertyValues {
       XCurve {keys [
         {arriveTangent=0.4876, tangentMode=RCTM_User, time=-0.0039, value=0.6566}
         {
          arriveTangent=1.1967
          arriveTangentWeight=0.5219
          interpMode=RCIM_Cubic
          leaveTangent=1.1967
          leaveTangentWeight=0.353
          tangentMode=RCTM_User
          time=1
          value=1
         }
        ]}
       YCurve {keys [
         {
          arriveTangent=0.225
          arriveTangentWeight=0.8393
          interpMode=RCIM_Cubic
          leaveTangent=0.225
          leaveTangentWeight=0.5346
          tangentMode=RCTM_User
          time=-0.0039
          value=0.6566
         }
         {
          arriveTangent=0.5054
          arriveTangentWeight=0.3749
          interpMode=RCIM_Cubic
          leaveTangent=0.5054
          leaveTangentWeight=0.3909
          tangentMode=RCTM_User
          time=1
          value=1
         }
        ]}
       ZCurve {keys [
         {
          arriveTangent=0.225
          arriveTangentWeight=0.8393
          interpMode=RCIM_Cubic
          leaveTangent=0.225
          leaveTangentWeight=0.5346
          tangentMode=RCTM_User
          time=-0.0039
          value=0.6566
         }
         {
          arriveTangent=0.5054
          arriveTangentWeight=0.3749
          interpMode=RCIM_Cubic
          leaveTangent=0.5054
          leaveTangentWeight=0.3909
          tangentMode=RCTM_User
          time=1
          value=1
         }
        ]}
      }}
     CurveIndex=$User.Charge Percent
    ))
   &ScaleColor(
    Scale Alpha=&FloatFromCurve(
     FloatCurve=DataInterface() {propertyValues {Curve {keys [
         {arriveTangent=2.9749, interpMode=RCIM_Cubic, leaveTangent=2.9749, tangentMode=RCTM_User, time=0, value=0}
         {arriveTangent=0.7537, interpMode=RCIM_Cubic, leaveTangent=0.7537, time=0.2085, value=0.6329}
         {interpMode=RCIM_Cubic, leaveTangent=-3.1778, time=1.0027, value=0.7557}
        ]}}}
     CurveIndex=$User.Charge Percent
    )
    Scale RGB=&MakeVectorFromLinearColorRGB(Color=&ColorFromCurve(
      ColorCurve=DataInterface() {propertyValues {
        AlphaCurve {keys []}
        BlueCurve {keys [<at 0.0646 0.0625 RCIM_Cubic>,<at 0.6542 0.2656 RCIM_Cubic>]}
        GreenCurve {keys [<at 0.0646 0.012 RCIM_Cubic>,<at 0.6542 0 RCIM_Cubic>]}
        RedCurve {keys [<at 0.0646 0.0158 RCIM_Cubic>,<at 0.6542 0.1181 RCIM_Cubic>]}
       }}
      CurveIndex=$User.Charge Percent
     ))
    ScaleA=true
    ScaleRGB=true
   )
   &DynamicMaterialParameters(
    Index 1 Param 1=&FloatFromCurve(
     FloatCurve=DataInterface() {propertyValues {Curve {keys [
         {arriveTangent=2.4188, interpMode=RCIM_Cubic, leaveTangent=2.4188, tangentMode=RCTM_User, time=0, value=-0.25}
         {arriveTangent=0.444, interpMode=RCIM_Cubic, leaveTangent=0.444, tangentMode=RCTM_User, time=1, value=1}
        ]}}}
     CurveIndex=$User.Charge Percent
    )
    Index 1 Param 2=&FloatFromCurve(FloatCurve=DataInterface() {propertyValues {Curve {keys [<at 0.05 1 RCIM_Cubic>,<at 0.25 0 RCIM_Cubic>]}}}, CurveIndex=$User.Charge Percent)
    Module.Param0WriteEnabled1=true
    Module.Param1WriteEnabled1=true
    Module.Param2WriteEnabled1=false
    Module.Param3WriteEnabled1=false
    Write Parameter Index 0=false
    Write Parameter Index 1=true
   ) {module=DynamicMaterialParameters001}
   &KillParticles(Kill Particles=&SetBoolByFloatComparison(A=0.99, B=$User.Charge Percent, Comparison Type=::NewEnumerator6))
  ]
  rendererClasses [&NiagaraMeshRendererProperties]
  renderers [{rendererClass=&NiagaraMeshRendererProperties, rendererIndex=0}]
 }
}
summary {
 systemName=NS_Projectile_ChargeUp
 emitters [
  {
   emitterName=Light_and_Flare
   rendererClasses [
    &NiagaraSpriteRendererProperties
    &NiagaraLightRendererProperties
    &NiagaraMeshRendererProperties
   ]
  }
  {emitterName=Shockwave01, rendererClasses [&NiagaraMeshRendererProperties]}
  {bEnabled=false, emitterName=Bright_Flash_Decal, rendererClasses [&NiagaraSpriteRendererProperties]}
  {emitterName=LW_Radial_Plasma_Energy, rendererClasses []}
  {emitterName=WindCone, rendererClasses [&NiagaraMeshRendererProperties]}
  {emitterName=HotDebris001, rendererClasses [&NiagaraSpriteRendererProperties]}
  {emitterName=LW_back, rendererClasses []}
  {emitterName=LW_Radial_Plasma_Energy001, rendererClasses []}
  {bEnabled=false, emitterName=LW_Radial_Plasma_Energy002, rendererClasses []}
  {emitterName=LW_Chunks, rendererClasses []}
 ]
 userVariables [User.Charge Percent {type=NiagaraFloat, defaultValue=0, description=""}]
}
